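# Widget Blueprint for a friend entry in the Activity Browser's server browser, displaying a single friend name.
# It derives from the native FortServerInfoFriendEntry class and consists only of the inherited Text_FriendName CommonTextBlock as its root, styled in bold Burbank Small with placeholder text.
# The event graph holds only empty event stubs, so any logic that fills in the name lies outside this Blueprint.
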WidgetBlueprint /FortniteCoreUI/Frontend/ActivityBrowser/ServerBrowserFriendEntry
graphs [EventGraph {nodes [
   Event PreConstruct()
   Event Construct()
   Event Tick()
  ]}]
variables []
widgets {
 info {
  inheritedWidgetCount=1
  parentClass=/Script/FortniteUI.FortServerInfoFriendEntry
  rootWidgetClass=&CommonTextBlock
  widgetCount=1
 }
 widgets [Text_FriendName {
   widgetClassPath=&CommonTextBlock
   bInherited=true
   widget=Text_FriendName
   properties {
    text=Friend
    font {
     bForceMonospaced=false
     bMaterialIsStencil=false
     fontMaterial=null
     fontObject=&BurbankSmall
     letterSpacing=0
     monospacedWidth=1
     size=16
     skewAmount=0
     typefaceFontName=Bold
     outlineSettings {
      bApplyOutlineToDropShadows=false
      bMiteredCorners=false
      bSeparateFillAlpha=false
      outlineMaterial=null
      outlineSize=0
      outlineColor=<rgba 0 0 0 1>
     }
    }
    shadowOffset=<0 0>
   }
  }]
}
