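# Translucent UI material that, by its name, supplies a gradient for the Creative server interface.
# A component mask takes the vertical (G) texture coordinate as a linear ramp, alongside a constant white color.
# It uses no textures or parameters, and the links from these nodes to the material outputs are not shown.
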Material /FortniteCoreUI/UI/Creative/Materials/M_UI_CreativeServerGradient
graph {
 properties {
  bUsedWithStaticMesh=false
  blendMode=BLEND_Translucent
  materialDomain=MD_UI
  pixelDepthOffsetMode=PDOM_Legacy
  usedShadingModels=Default Lit
 }
 nodes [
  ComponentMask(None=TextureCoordinate()) {g=true}
  Constant3Vector() {constant=<rgba 1 1 1 0>}
 ]
}
