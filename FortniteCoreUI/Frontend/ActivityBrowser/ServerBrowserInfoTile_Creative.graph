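# Creative server browser tile, based on FortCreativeServerInfoTile, showing a server's artwork, owner and launch state.
# A hash-derived index picks one of twelve display-option data table rows, async-loading its texture and tinting the fill and an HSV-darkened gradient.
# Selection and hover drive size, border and overlay changes, with hover skipped for touch, selected or mid-animation tiles.
# Show/close friends buttons play ViewPlayers forward and back, and deselection reverses an open players view.
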WidgetBlueprint /FortniteCoreUI/Frontend/ActivityBrowser/ServerBrowserInfoTile_Creative
animations [
 Selected {
  binding_count=5
  length_seconds=1
  object=&self:Selected
  sequence {bindings [
    SizeBox_Root {
     id=#1
     possessed_class=SizeBox
     tracks [
      {
       type=MovieSceneFloatTrack
       display_name=Width Override
       sections [{type=MovieSceneFloatSection, channels [None {type=MovieSceneScriptingFloatChannel, keys [[0,315.4871],[20,435]]}]}]
      }
      {
       type=MovieScene2DTransformTrack
       sections [{
         type=MovieScene2DTransformSection
         end_frame=20
         start_frame=0
         channels [
          Translation.X {type=MovieSceneScriptingFloatChannel}
          Translation.Y {type=MovieSceneScriptingFloatChannel}
          Angle {type=MovieSceneScriptingFloatChannel}
          Scale.X {type=MovieSceneScriptingFloatChannel, keys [[0,0.8],[15,1.0449],[19,0.9924],[20,1]]}
          Scale.Y {type=MovieSceneScriptingFloatChannel, keys [[0,0.8],[15,1.0449],[19,0.9924],[20,1]]}
          Shear.X {type=MovieSceneScriptingFloatChannel}
          Shear.Y {type=MovieSceneScriptingFloatChannel}
         ]
        }]
      }
     ]
    }
    BorderSlot_0 {
     id=#2
     parent_id=#1
     possessed_class=BorderSlot
     tracks [{
       type=MovieSceneMarginTrack
       display_name=Padding
       sections [{
         type=MovieSceneMarginSection
         end_frame=20
         start_frame=0
         channels [
          Left {type=MovieSceneScriptingFloatChannel, keys [[0,-26.0614],[20,13.6259]]}
          Top {type=MovieSceneScriptingFloatChannel, const [0,2]}
          Right {type=MovieSceneScriptingFloatChannel, keys [[0,-26.0438],[20,12]]}
          Bottom {type=MovieSceneScriptingFloatChannel, const [0,2]}
         ]
        }]
      }]
    }
    Image_ColorFill {
     id=#3
     possessed_class=Image
     tracks [{
       type=MovieSceneFloatTrack
       display_name=Render Opacity
       sections [{type=MovieSceneFloatSection, end_frame=20, start_frame=0, channels [None {type=MovieSceneScriptingFloatChannel, keys [[0,0.45],[20,0]]}]}]
      }]
    }
    Image_PanelBorder {
     id=#4
     possessed_class=Image
     tracks [{
       type=MovieSceneWidgetMaterialTrack
       sections [{
         type=MovieSceneParameterSection
         channels [
          BottomEdge_1 {type=MovieSceneScriptingFloatChannel, keys [[0,-0.0222],[20,0.0583]]}
          LeftEdge_1 {type=MovieSceneScriptingFloatChannel, keys [[0,-0.009],[20,0.024]]}
          RightEdge_1 {type=MovieSceneScriptingFloatChannel, keys [[0,0],[20,0.027]]}
          TopEdge_1 {type=MovieSceneScriptingFloatChannel, keys [[0,0],[20,0.081]]}
         ]
        }]
      }]
    }
    Button_ShowFriends {
     id=#5
     possessed_class=SimpleCommonButton_C
     tracks [{
       type=MovieSceneByteTrack
       display_name=Visibility
       sections [{type=MovieSceneByteSection, channels [None {type=MovieSceneScriptingByteChannel, keys [[0,1],[1.994,0]]}]}]
      }]
    }
   ]}
 }
 ViewPlayers {
  binding_count=18
  length_seconds=0.3676
  object=&self:ViewPlayers
  sequence {
   bindings [
    Button_ShowFriends {
     id=#6
     possessed_class=SimpleCommonButton_C
     tracks [{
       type=MovieSceneByteTrack
       display_name=Visibility
       sections [{type=MovieSceneByteSection, channels [None {type=MovieSceneScriptingByteChannel, keys [[0,0],[1.306,1]]}]}]
      }]
    }
    CommonBorderBackButton {
     id=#7
     possessed_class=CommonBorder
     tracks [
      {
       type=MovieSceneFloatTrack
       display_name=Render Opacity
       sections [{type=MovieSceneFloatSection, channels [None {type=MovieSceneScriptingFloatChannel, keys [[0,0],[44.112,1]]}]}]
      }
      {
       type=MovieSceneFloatVectorTrack
       display_name=Desired Size Scale
       sections [{
         type=MovieSceneFloatVectorSection
         channels [
          Vector.X {type=MovieSceneScriptingFloatChannel}
          Vector.Y {type=MovieSceneScriptingFloatChannel, keys [[0,0],[44.112,1]]}
         ]
        }]
      }
     ]
    }
    BackgroundBlur_0 {
     id=#8
     possessed_class=BackgroundBlur
     tracks [{
       type=MovieSceneFloatTrack
       display_name=Blur Strength
       sections [{type=MovieSceneFloatSection, channels [None {type=MovieSceneScriptingFloatChannel, keys [[0,0],[44.112,10]]}]}]
      }]
    }
    Button_CloseFriends {
     id=#9
     possessed_class=SimpleCommonButton_C
     tracks [{
       type=MovieSceneByteTrack
       display_name=Visibility
       sections [{type=MovieSceneByteSection, channels [None {type=MovieSceneScriptingByteChannel, keys [[1.306,1],[2.562,0]]}]}]
      }]
    }
    Image_112 {
     id=#10
     possessed_class=Image
     tracks [{
       type=MovieScene2DTransformTrack
       sections [{
         type=MovieScene2DTransformSection
         end_frame=44
         start_frame=0
         channels [
          Translation.X {type=MovieSceneScriptingFloatChannel}
          Translation.Y {type=MovieSceneScriptingFloatChannel}
          Angle {type=MovieSceneScriptingFloatChannel}
          Scale.X {type=MovieSceneScriptingFloatChannel, keys [[0,0],[44.094,-1]]}
          Scale.Y {type=MovieSceneScriptingFloatChannel, keys [[0,0],[44.094,1]]}
          Shear.X {type=MovieSceneScriptingFloatChannel}
          Shear.Y {type=MovieSceneScriptingFloatChannel}
         ]
        }]
      }]
    }
    Image_116 {
     id=#11
     possessed_class=Image
     tracks [{
       type=MovieScene2DTransformTrack
       sections [{
         type=MovieScene2DTransformSection
         end_frame=44
         start_frame=0
         channels [
          Translation.X {type=MovieSceneScriptingFloatChannel}
          Translation.Y {type=MovieSceneScriptingFloatChannel}
          Angle {type=MovieSceneScriptingFloatChannel}
          Scale.X {type=MovieSceneScriptingFloatChannel, keys [[0,0],[44.098,-1]]}
          Scale.Y {type=MovieSceneScriptingFloatChannel, keys [[0,0],[44.098,-1]]}
          Shear.X {type=MovieSceneScriptingFloatChannel}
          Shear.Y {type=MovieSceneScriptingFloatChannel}
         ]
        }]
      }]
    }
    ScaleBox_3 {
     id=#12
     display_name=ScrollBox_2
     possessed_class=ScaleBox
     tracks [{
       type=MovieSceneFloatTrack
       display_name=Render Opacity
       sections [{type=MovieSceneFloatSection, channels [None {type=MovieSceneScriptingFloatChannel, keys [[0,0],[30,1],[44.106,1]]}]}]
      }]
    }
    Image_ColorFill {
     id=#13
     possessed_class=Image
     tracks [{
       type=MovieSceneFloatTrack
       display_name=Render Opacity
       sections [{type=MovieSceneFloatSection, channels [None {type=MovieSceneScriptingFloatChannel, keys [[0,0],[44.112,0.9]]}]}]
      }]
    }
    CommonBorder_1 {
     id=#14
     possessed_class=CommonBorder
     tracks [{
       type=MovieSceneFloatVectorTrack
       display_name=Desired Size Scale
       sections [{
         type=MovieSceneFloatVectorSection
         channels [
          Vector.X {type=MovieSceneScriptingFloatChannel}
          Vector.Y {type=MovieSceneScriptingFloatChannel, keys [[0,0],[36,1.0011],[44.106,1]]}
         ]
        }]
      }]
    }
    OverlaySpikes {
     id=#15
     possessed_class=Overlay
     tracks [{
       type=MovieSceneFloatTrack
       display_name=Render Opacity
       sections [{type=MovieSceneFloatSection, channels [None {type=MovieSceneScriptingFloatChannel, keys [[0,0],[44.102,1]]}]}]
      }]
    }
    Image_114 {
     id=#16
     possessed_class=Image
     tracks [{
       type=MovieScene2DTransformTrack
       sections [{
         type=MovieScene2DTransformSection
         end_frame=44
         start_frame=0
         channels [
          Translation.X {type=MovieSceneScriptingFloatChannel}
          Translation.Y {type=MovieSceneScriptingFloatChannel}
          Angle {type=MovieSceneScriptingFloatChannel}
          Scale.X {type=MovieSceneScriptingFloatChannel, keys [[0,0],[44.106,1]]}
          Scale.Y {type=MovieSceneScriptingFloatChannel, keys [[0,0],[44.106,-1]]}
          Shear.X {type=MovieSceneScriptingFloatChannel}
          Shear.Y {type=MovieSceneScriptingFloatChannel}
         ]
        }]
      }]
    }
    CommonBorderInspectTitle {
     id=#17
     possessed_class=CommonBorder
     tracks [
      {
       type=MovieSceneFloatTrack
       display_name=Render Opacity
       sections [{type=MovieSceneFloatSection, channels [None {type=MovieSceneScriptingFloatChannel, keys [[0,0],[11.916,1]]}]}]
      }
      {
       type=MovieSceneFloatVectorTrack
       display_name=Desired Size Scale
       sections [{
         type=MovieSceneFloatVectorSection
         channels [
          Vector.X {type=MovieSceneScriptingFloatChannel}
          Vector.Y {type=MovieSceneScriptingFloatChannel, keys [[0,0],[44.112,1]]}
         ]
        }]
      }
     ]
    }
    spike {
     id=#18
     possessed_class=Image
     tracks [{
       type=MovieScene2DTransformTrack
       sections [{
         type=MovieScene2DTransformSection
         end_frame=44
         start_frame=0
         channels [
          Translation.X {type=MovieSceneScriptingFloatChannel}
          Translation.Y {type=MovieSceneScriptingFloatChannel}
          Angle {type=MovieSceneScriptingFloatChannel}
          Scale.X {type=MovieSceneScriptingFloatChannel, keys [[0,0],[44.106,1]]}
          Scale.Y {type=MovieSceneScriptingFloatChannel, keys [[0,0],[44.106,1]]}
          Shear.X {type=MovieSceneScriptingFloatChannel}
          Shear.Y {type=MovieSceneScriptingFloatChannel}
         ]
        }]
      }]
    }
    CommonBorderDefaultTitle {
     id=#19
     possessed_class=CommonBorder
     tracks [{
       type=MovieSceneFloatVectorTrack
       display_name=Desired Size Scale
       sections [{
         type=MovieSceneFloatVectorSection
         channels [
          Vector.X {type=MovieSceneScriptingFloatChannel}
          Vector.Y {type=MovieSceneScriptingFloatChannel, keys [[0,1],[33.358,0],[44.112,0]]}
         ]
        }]
      }]
    }
    Image_115 {
     id=#20
     possessed_class=Image
     tracks [{
       type=MovieScene2DTransformTrack
       sections [{
         type=MovieScene2DTransformSection
         end_frame=44
         start_frame=0
         channels [
          Translation.X {type=MovieSceneScriptingFloatChannel}
          Translation.Y {type=MovieSceneScriptingFloatChannel}
          Angle {type=MovieSceneScriptingFloatChannel}
          Scale.X {type=MovieSceneScriptingFloatChannel, keys [[0,0],[44.106,1]]}
          Scale.Y {type=MovieSceneScriptingFloatChannel, keys [[0,0],[44.106,-1]]}
          Shear.X {type=MovieSceneScriptingFloatChannel}
          Shear.Y {type=MovieSceneScriptingFloatChannel}
         ]
        }]
      }]
    }
    Image_Gradient {
     id=#21
     possessed_class=Image
     tracks [{
       type=MovieSceneFloatTrack
       display_name=Render Opacity
       sections [{type=MovieSceneFloatSection, channels [None {type=MovieSceneScriptingFloatChannel, keys [[0,1],[44.106,0.5]]}]}]
      }]
    }
    OwnershipBorder {
     id=#22
     possessed_class=Border
     tracks [{
       type=MovieSceneFloatVectorTrack
       display_name=Desired Size Scale
       sections [{
         type=MovieSceneFloatVectorSection
         end_frame=12
         start_frame=0
         channels [
          Vector.X {type=MovieSceneScriptingFloatChannel, const [1,2]}
          Vector.Y {type=MovieSceneScriptingFloatChannel, keys [[0,1],[12,0]]}
         ]
        }]
      }]
    }
    Image_113 {
     id=#23
     possessed_class=Image
     tracks [{
       type=MovieScene2DTransformTrack
       sections [{
         type=MovieScene2DTransformSection
         end_frame=44
         start_frame=0
         channels [
          Translation.X {type=MovieSceneScriptingFloatChannel}
          Translation.Y {type=MovieSceneScriptingFloatChannel}
          Angle {type=MovieSceneScriptingFloatChannel}
          Scale.X {type=MovieSceneScriptingFloatChannel, keys [[0,0],[44.106,-1]]}
          Scale.Y {type=MovieSceneScriptingFloatChannel, keys [[0,0],[44.106,1]]}
          Shear.X {type=MovieSceneScriptingFloatChannel}
          Shear.Y {type=MovieSceneScriptingFloatChannel}
         ]
        }]
      }]
    }
   ]
   display_rate {numerator=120}
  }
 }
 Selected_1 {
  binding_count=6
  length_seconds=1
  object=&self:Selected_1
  sequence {bindings [
    SizeBox_Root {
     id=#1
     possessed_class=SizeBox
     tracks [
      {
       type=MovieSceneFloatTrack
       display_name=Width Override
       sections [{type=MovieSceneFloatSection, channels [None {type=MovieSceneScriptingFloatChannel, keys [[0,384],[20.0003,435]]}]}]
      }
      {
       type=MovieScene2DTransformTrack
       sections [{
         type=MovieScene2DTransformSection
         end_frame=20
         start_frame=0
         channels [
          Translation.X {type=MovieSceneScriptingFloatChannel}
          Translation.Y {type=MovieSceneScriptingFloatChannel}
          Angle {type=MovieSceneScriptingFloatChannel}
          Scale.X {type=MovieSceneScriptingFloatChannel, keys [[0,0.8],[20,1]]}
          Scale.Y {type=MovieSceneScriptingFloatChannel, keys [[0,0.8],[20,1]]}
          Shear.X {type=MovieSceneScriptingFloatChannel}
          Shear.Y {type=MovieSceneScriptingFloatChannel}
         ]
        }]
      }
     ]
    }
    BorderSlot_0 {
     id=#2
     parent_id=#1
     possessed_class=BorderSlot
     tracks [{
       type=MovieSceneMarginTrack
       display_name=Padding
       sections [{
         type=MovieSceneMarginSection
         end_frame=20
         start_frame=0
         channels [
          Left {type=MovieSceneScriptingFloatChannel, keys [[0,-26.0614],[20,12]]}
          Top {type=MovieSceneScriptingFloatChannel, const [0,2]}
          Right {type=MovieSceneScriptingFloatChannel, keys [[0,-26.0438],[20,12]]}
          Bottom {type=MovieSceneScriptingFloatChannel, const [0,2]}
         ]
        }]
      }]
    }
    Image_ColorFill {
     id=#3
     possessed_class=Image
     tracks [{
       type=MovieSceneFloatTrack
       display_name=Render Opacity
       sections [{type=MovieSceneFloatSection, end_frame=20, start_frame=0, channels [None {type=MovieSceneScriptingFloatChannel, keys [[0,0.45],[20,0]]}]}]
      }]
    }
    Image_PanelBorder {
     id=#4
     possessed_class=Image
     tracks [
      {
       type=MovieSceneWidgetMaterialTrack
       sections [{
         type=MovieSceneParameterSection
         channels [
          BottomEdge_1 {type=MovieSceneScriptingFloatChannel, keys [[0,0.05],[20,0.0825]]}
          LeftEdge_1 {type=MovieSceneScriptingFloatChannel, keys [[0,0.0175],[20,0.024]]}
          RightEdge_1 {type=MovieSceneScriptingFloatChannel, keys [[0,0.02],[20,0.027]]}
          TopEdge_1 {type=MovieSceneScriptingFloatChannel, keys [[0,0.075],[20,0.1]]}
         ]
        }]
      }
      {
       type=MovieSceneFloatTrack
       display_name=Render Opacity
       sections [{type=MovieSceneFloatSection, end_frame=20, start_frame=0, channels [None {type=MovieSceneScriptingFloatChannel}]}]
      }
      {
       type=MovieSceneColorTrack
       display_name=Color and Opacity
       sections [{
         type=MovieSceneColorSection
         end_frame=20
         start_frame=0
         channels [
          Color.R {type=MovieSceneScriptingFloatChannel, keys [[0,0],[20,1]]}
          Color.G {type=MovieSceneScriptingFloatChannel, keys [[0,0.4],[20,1]]}
          Color.B {type=MovieSceneScriptingFloatChannel, const [1,1]}
          Color.A {type=MovieSceneScriptingFloatChannel, const [1,2]}
         ]
        }]
      }
     ]
    }
    SizeBox_0 {
     id=#24
     possessed_class=SizeBox
     tracks [{
       type=MovieSceneFloatTrack
       display_name=Height Override
       sections [{
         type=MovieSceneFloatSection
         end_frame=20
         start_frame=0
         channels [None {type=MovieSceneScriptingFloatChannel, keys [[0,0],[15,68],[20,64]]}]
        }]
      }]
    }
    Button_ShowFriends {
     id=#5
     possessed_class=SimpleCommonButton_C
     tracks [{
       type=MovieSceneByteTrack
       display_name=Visibility
       sections [{type=MovieSceneByteSection, channels [None {type=MovieSceneScriptingByteChannel, keys [[0,1],[1.994,0]]}]}]
      }]
    }
   ]}
 }
 Hover {
  binding_count=5
  length_seconds=1
  object=&self:Hover
  sequence {bindings [
    SizeBox_Root {
     id=#1
     possessed_class=SizeBox
     tracks [
      {
       type=MovieSceneFloatTrack
       display_name=Width Override
       sections [{type=MovieSceneFloatSection, channels [None {type=MovieSceneScriptingFloatChannel, const [384,1]}]}]
      }
      {
       type=MovieScene2DTransformTrack
       sections [{
         type=MovieScene2DTransformSection
         end_frame=20
         start_frame=0
         channels [
          Translation.X {type=MovieSceneScriptingFloatChannel}
          Translation.Y {type=MovieSceneScriptingFloatChannel}
          Angle {type=MovieSceneScriptingFloatChannel}
          Scale.X {type=MovieSceneScriptingFloatChannel, const [0.8,1]}
          Scale.Y {type=MovieSceneScriptingFloatChannel, const [0.8,1]}
          Shear.X {type=MovieSceneScriptingFloatChannel}
          Shear.Y {type=MovieSceneScriptingFloatChannel}
         ]
        }]
      }
     ]
    }
    BorderSlot_0 {
     id=#2
     parent_id=#1
     possessed_class=BorderSlot
     tracks [{
       type=MovieSceneMarginTrack
       display_name=Padding
       sections [{
         type=MovieSceneMarginSection
         end_frame=20
         start_frame=0
         channels [
          Left {type=MovieSceneScriptingFloatChannel, const [-26.0614,1]}
          Top {type=MovieSceneScriptingFloatChannel, const [0,1]}
          Right {type=MovieSceneScriptingFloatChannel, const [-26.0438,1]}
          Bottom {type=MovieSceneScriptingFloatChannel, const [0,1]}
         ]
        }]
      }]
    }
    Image_ColorFill {
     id=#3
     possessed_class=Image
     tracks [{
       type=MovieSceneFloatTrack
       display_name=Render Opacity
       sections [{type=MovieSceneFloatSection, end_frame=20, start_frame=0, channels [None {type=MovieSceneScriptingFloatChannel, keys [[0,0.45],[20,0.25]]}]}]
      }]
    }
    Image_PanelBorder {
     id=#4
     possessed_class=Image
     tracks [
      {
       type=MovieSceneWidgetMaterialTrack
       sections [{
         type=MovieSceneParameterSection
         channels [
          BottomEdge_1 {type=MovieSceneScriptingFloatChannel, const [0.05,1]}
          LeftEdge_1 {type=MovieSceneScriptingFloatChannel, const [0.0175,1]}
          RightEdge_1 {type=MovieSceneScriptingFloatChannel, const [0.02,1]}
          TopEdge_1 {type=MovieSceneScriptingFloatChannel, const [0.075,1]}
         ]
        }]
      }
      {
       type=MovieSceneFloatTrack
       display_name=Render Opacity
       sections [{type=MovieSceneFloatSection, end_frame=20, start_frame=0, channels [None {type=MovieSceneScriptingFloatChannel}]}]
      }
      {
       type=MovieSceneColorTrack
       display_name=Color and Opacity
       sections [{
         type=MovieSceneColorSection
         end_frame=20
         start_frame=0
         channels [
          Color.R {type=MovieSceneScriptingFloatChannel, const [0,1]}
          Color.G {type=MovieSceneScriptingFloatChannel, keys [[0,0.4],[20,0.6666]]}
          Color.B {type=MovieSceneScriptingFloatChannel, const [1,1]}
          Color.A {type=MovieSceneScriptingFloatChannel, const [1,2]}
         ]
        }]
      }
     ]
    }
    Button_ShowFriends {
     id=#5
     possessed_class=SimpleCommonButton_C
     tracks [{
       type=MovieSceneByteTrack
       display_name=Visibility
       sections [{type=MovieSceneByteSection, channels [None {type=MovieSceneScriptingByteChannel, const [1,1]}]}]
      }]
    }
   ]}
 }
 ShowCreatorName {
  binding_count=2
  length_seconds=0.35
  object=&self:ShowCreatorName
  sequence {bindings [
    CreatedByText {
     id=#25
     possessed_class=RichTextBlock
     tracks [{
       type=MovieScene2DTransformTrack
       sections [{
         type=MovieScene2DTransformSection
         channels [
          Translation.X {type=MovieSceneScriptingFloatChannel}
          Translation.Y {type=MovieSceneScriptingFloatChannel, keys [[2,-128],[7,0]]}
          Angle {type=MovieSceneScriptingFloatChannel}
          Scale.X {type=MovieSceneScriptingFloatChannel}
          Scale.Y {type=MovieSceneScriptingFloatChannel}
          Shear.X {type=MovieSceneScriptingFloatChannel}
          Shear.Y {type=MovieSceneScriptingFloatChannel}
         ]
        }]
      }]
    }
    Text_OwnerName {
     id=#26
     possessed_class=CommonTextBlock
     tracks [{
       type=MovieScene2DTransformTrack
       sections [{
         type=MovieScene2DTransformSection
         channels [
          Translation.X {type=MovieSceneScriptingFloatChannel}
          Translation.Y {type=MovieSceneScriptingFloatChannel, keys [[0,-128],[5,0]]}
          Angle {type=MovieSceneScriptingFloatChannel}
          Scale.X {type=MovieSceneScriptingFloatChannel, keys [[0,0.75],[5,1]]}
          Scale.Y {type=MovieSceneScriptingFloatChannel, keys [[0,0.75],[5,1]]}
          Shear.X {type=MovieSceneScriptingFloatChannel}
          Shear.Y {type=MovieSceneScriptingFloatChannel}
         ]
        }]
      }]
    }
   ]}
 }
]
graphs [EventGraph {nodes [
   "On Button Clicked (Button_ShowFriends)"!() -> (OutputDelegate, then, Button:Common Button Legacy Object Reference) {is=title K2Node_ComponentBoundEvent}
   n1 = Get ViewPlayers() -> (ViewPlayers:Widget Animation Object Reference)
   > PlayAnimationForward(InAnimation=@n1)
   > Set PlayersShown(PlayersShown:bool=true) -> (Output_Get:bool)
   n2 = Get Button_ShowFriends() -> (Button_ShowFriends:Simple Common Button With Input Object Reference)
   > SetVisibility(self=@n2, InVisibility=Collapsed)
   n3 = Get Button_CloseFriends() -> (Button_CloseFriends:Simple Common Button With Input Object Reference)
   > SetVisibility(self=@n3, InVisibility=Visible)
   "On Button Clicked (Button_CloseFriends)"!() -> (OutputDelegate, then, Button:Common Button Legacy Object Reference) {is=title K2Node_ComponentBoundEvent}
   > PlayAnimationReverse(InAnimation=@n1)
   > Set PlayersShown(PlayersShown=false)
   > SetVisibility(self=@n2, InVisibility=Visible)
   > SetVisibility(self=@n3, InVisibility=Collapsed)
   n4 = InitializeSelectedButton() {is=title K2Node_CustomEvent}
   n5 = Get Selected_1() -> (Selected_1:Widget Animation Object Reference)
   PlayAnimationForward(execute=[@n23.then, @n4.then], InAnimation=@n5, PlaybackSpeed=6)
   n6 = Get Button_ShowFriends()
   > SetVisibility(self=@n6, InVisibility=Visible)
   n7 = SetBackgroundImage!() -> (OutputDelegate, then, ArrayIndex:int) {is=title K2Node_CustomEvent}
   n8 = "To String (Integer)"(InInt=int + int(A="% (Integer)"(A=@n7.ArrayIndex, B=12) -> (ReturnValue=0), B=1) -> (ReturnValue=0) {is=title K2Node_CommutativeAssociativeBinaryOperator})
   n9 = MakeLiteralString(Value=Option)
   > n10 = Get Data Table Row CreativeServerDisplayOptions!(
    execute:Exec
    DataTable:Data Table Object Reference=/FortniteCoreUI/UI/Creative/CreativeServerDisplayOptions
    RowName:Name=String To Name(InString=Append(A=@n9, B=@n8))
   ) -> (then:Exec, RowNotFound:Exec, ReturnValue:Fort Creative Server Display Option Structure) {is=title K2Node_GetDataTableRow}
   n11 = Break Fort Creative Server Display Option!(FortCreativeServerDisplayOption:"Fort Creative Server Display Option Structure (by ref)"=@n10) -> (Image:Texture Soft Object Reference, Color:Linear Color Structure) {is=title K2Node_BreakStruct}
   > n12 = Async Load Asset(Asset=@n11)
   n13 = Cast To Texture2D(execute=@n12.Completed, Object=@n12)
   > SetBrushFromTexture!(
    execute
    self:Image Object Reference=Get Image_ServerImage() -> (Image_ServerImage:Image Object Reference)
    Texture:Texture 2D Object Reference=@n13
    bMatchSize:bool=true
   ) -> (then)
   > SetColorAndOpacity(
    self:Image Object Reference=Get Image_ColorFill() -> (Image_ColorFill:Image Object Reference)
    InColorAndOpacity:Linear Color Structure=@n11.Color
   )
   n14 = RGB to HSV!(InColor:Linear Color Structure=@n11.Color) -> (H:float, S:float, V:float, A:float)
   > SetColorAndOpacity(
    self=Get Image_Gradient() -> (Image_Gradient:Image Object Reference)
    InColorAndOpacity=HSV to RGB!(H:float=@n14, S:float=float * float(A=@n14.S, B=0.8), V:float=float * float(A=@n14.V, B=0.8), A:float=@n14.A) -> (ReturnValue:Linear Color Structure)
   )
   n15 = Event RequestUpdateBackgroundImage!() -> (OutputDelegate, then, FriendListHash:int=0) {is=title K2Node_Event}
   > SetBackgroundImage!(execute, self, ArrayIndex:int=@n15.FriendListHash) -> (then) {is=title K2Node_CallFunction}
   Event Construct()
   n16 = Get Selected_1()
   > n17 = PlayAnimation(InAnimation=@n16, StartAtTime=GetEndTime(self=@n16), PlayMode=Reverse, PlaybackSpeed=3)
   n18 = InputMethodChanged!() -> (OutputDelegate, then, bNewInputType:ECommonInputType Enum) {is=title K2Node_CustomEvent}
   Bind Event to On Input Method Changed(execute=@n17.then, self=Get CommonInputSubsystem(), Delegate=@n18)
   > InputMethodChanged!(execute, self, bNewInputType:ECommonInputType Enum=MouseAndKeyboard) -> (then) {is=title K2Node_CallFunction}
   n19 = Event OnLaunchButtonEnableChanged!() -> (OutputDelegate, then, bEnableValue:bool) {is=title K2Node_Event}
   > n20 = Cast To PrimarySkewButton!(execute, Object=Get Button_Launch() -> (Button_Launch:Common Button Legacy Object Reference)) -> (then, CastFailed, AsPrimary Skew Button:Primary Skew Button Object Reference) {is=title K2Node_DynamicCast}
   > "SetText (Text)"(
    self=Get CenterButtonTextWidget!(self:Primary Skew Button Object Reference=@n20) -> (CenterButtonTextWidget:Common Text Object Reference)
    InText=Select(Option 0:Text="LOCTEXT(\"DISABLED\")", Option 1:Text="LOCTEXT(\"LAUNCH\")", Index=@n19.bEnableValue) -> (ReturnValue:Text)
   )
   Event On Selected()
   n21 = Get Launch_Overlay() -> (Launch_Overlay:Scale Box Object Reference)
   > SetVisibility(self=@n21, InVisibility=SelfHitTestInvisible)
   n22 = Get WarningOverlay() -> (WarningOverlay:Overlay Object Reference)
   > n23 = SetVisibility(self=@n22, InVisibility=SelfHitTestInvisible)
   Event On Deselected()
   > SetVisibility(self=@n21, InVisibility=Collapsed)
   > SetVisibility(self=@n22, InVisibility=Collapsed)
   > PlayAnimationReverse(InAnimation=@n5, PlaybackSpeed=6)
   > SetVisibility(self=Get Button_CloseFriends(), InVisibility=Collapsed)
   > SetVisibility(self=@n6, InVisibility=Collapsed)
   > Branch(Condition=Get PlayersShown() -> (PlayersShown:bool))
   > PlayAnimationReverse(InAnimation=Get ViewPlayers(), PlaybackSpeed=4)
   Event On Hovered()
   n24 = IsUsingTouch(PlayerController=GetOwningPlayer())
   > n25 = Branch(Condition=@n24)
   n26 = GetSelected!(self:Common Button Base Object Reference) -> (ReturnValue:bool)
   n27 = Branch(execute=@n25.else, Condition=@n26)
   n28 = IsAnimationPlaying(InAnimation=Get Selected_1())
   n29 = Branch(execute=@n27.else, Condition=@n28)
   n30 = Get Hover() -> (Hover:Widget Animation Object Reference)
   PlayAnimationForward(execute=@n29.else, InAnimation=@n30, PlaybackSpeed=4)
   Event On Unhovered()
   > n31 = Branch(Condition=@n24)
   n32 = Branch(execute=@n31.else, Condition=@n26)
   n33 = Branch(execute=@n32.else, Condition=@n28)
   PlayAnimationReverse(execute=@n33.else, InAnimation=@n30, PlaybackSpeed=4)
   Event OnOwnerNameUpdated() {is=title K2Node_Event}
   > PlayAnimation(InAnimation=Get ShowCreatorName() -> (ShowCreatorName:Widget Animation Object Reference))
   n34 = RGB to HSV(InColor=@n11.Color)
   HSV to RGB(H=@n34, S=0.25, V=1, A=@n34.A)
   SetStyle!(
    execute
    self:Common Button Base Object Reference=Get Button_ShowFriends()
    InStyle:Common Button Style Class Reference=Select(
     Option 0:Common Button Style Class Reference=/FortUILibrary/Foundation/Buttons/ButtonStyles/ButtonStyle-Serverbrowser.ButtonStyle-Serverbrowser_C
     Option 1:Common Button Style Class Reference=&ButtonStyle-Serverbrowser_C
     Index="Equal (Enum)"(A=@n18.bNewInputType, B=Gamepad)
    ) -> (ReturnValue:Common Button Style Class Reference)
   ) -> (then)
  ]}]
variables [
 FriendListOutput {type="array<string>", default []}
 NewServerIsPrivate {type=boolean}
 PlayersShown {type=boolean}
]
widgets {
 info {inheritedWidgetCount=9, parentClass=/Script/FortniteUI.FortCreativeServerInfoTile, rootWidgetClass=&ScaleBox, widgetCount=58}
 widgets [
  ScaleBox_1 {widgetClassPath=&ScaleBox, widget=ScaleBox_1, properties {stretch=UserSpecified, userSpecifiedScale=1.2}}
  CommonBorder_4 {
   widgetClassPath=&CommonBorder
   parent=ScaleBox_1
   widget=CommonBorder_4
   properties {horizontalAlignment=HAlign_Center, verticalAlignment=VAlign_Center, padding=<ltrb 12 0 12 12>}
   slot {class=ScaleBoxSlot}
  }
  SizeBox_Root {
   widgetClassPath=&SizeBox
   parent=CommonBorder_4
   widget=SizeBox_Root
   properties {bOverride_HeightOverride=true, bOverride_WidthOverride=true, heightOverride=650, widthOverride=364.3404}
   slot {class=BorderSlot, properties {horizontalAlignment=HAlign_Center, verticalAlignment=VAlign_Center, padding=<ltrb 12 0 12 12>}}
  }
  Overlay_0 {widgetClassPath=&Overlay, parent=SizeBox_Root, widget=Overlay_0, properties {clipping=ClipToBounds}, slot {class=SizeBoxSlot}}
  SizeBox_1 {
   widgetClassPath=&SizeBox
   parent=Overlay_0
   widget=SizeBox_1
   properties {bOverride_HeightOverride=true, bOverride_WidthOverride=true, heightOverride=650, widthOverride=435}
   slot {class=OverlaySlot, properties {horizontalAlignment=HAlign_Center, verticalAlignment=VAlign_Center, padding=<ltrb -100 0 -100 0>}}
  }
  Overlay_2 {widgetClassPath=&Overlay, parent=SizeBox_1, widget=Overlay_2, properties {clipping=ClipToBounds}, slot {class=SizeBoxSlot}}
  Image_ServerImage {
   widgetClassPath=&Image
   bIsVariable=true
   parent=Overlay_2
   widget=Image_ServerImage
   properties {brush {
     bIsDynamicallyLoaded=false
     drawAs=Image
     imageType=NoImage
     mirroring=NoMirror
     resourceName=None
     resourceObject=/FortniteCoreUI/UI/Creative/Materials/ServerImageAssets/T_UI_Creative_ServerImage_7
     tiling=NoTile
     imageSize=<512 1024>
     margin=<ltrb 0 0 0 0>
     outlineSettings {
      bUseBrushTransparency=false
      roundingType=HalfHeightRadius
      width=0
      color {colorUseRule=UseColor_Specified, specifiedColor=<rgba 0 0 0 0>}
      cornerRadii=<0 0 0 0>
     }
     tintColor {colorUseRule=UseColor_Specified, specifiedColor=<rgba 1 1 1 1>}
     uVRegion {bIsValid=false, max=<0 0>, min=<0 0>}
    }}
   slot {class=OverlaySlot, properties {horizontalAlignment=HAlign_Fill, verticalAlignment=VAlign_Fill}}
  }
  BackgroundBlur_0 {
   widgetClassPath=&BackgroundBlur
   parent=Overlay_2
   widget=BackgroundBlur_0
   slot {class=OverlaySlot, properties {horizontalAlignment=HAlign_Fill, verticalAlignment=VAlign_Fill}}
  }
  Image_ColorFill {
   widgetClassPath=&Image
   bIsVariable=true
   parent=Overlay_2
   widget=Image_ColorFill
   properties {renderOpacity=0, colorAndOpacity=<rgba 0.05 0.2083 1 1>}
   slot {class=OverlaySlot, properties {horizontalAlignment=HAlign_Fill, verticalAlignment=VAlign_Fill}}
  }
  Image_Gradient {
   widgetClassPath=&Image
   bIsVariable=true
   parent=Overlay_2
   widget=Image_Gradient
   properties {
    brush {
     bIsDynamicallyLoaded=false
     drawAs=Image
     imageType=NoImage
     mirroring=NoMirror
     resourceName=None
     resourceObject=/FortniteCoreUI/UI/Creative/Materials/M_UI_CreativeServerGradient
     tiling=NoTile
     imageSize=<32 340.801>
     margin=<ltrb 0 0 0 0>
     outlineSettings {
      bUseBrushTransparency=false
      roundingType=HalfHeightRadius
      width=0
      color {colorUseRule=UseColor_Specified, specifiedColor=<rgba 0 0 0 0>}
      cornerRadii=<0 0 0 0>
     }
     tintColor {colorUseRule=UseColor_Specified, specifiedColor=<rgba 1 1 1 1>}
     uVRegion {bIsValid=false, max=<0 0>, min=<0 0>}
    }
    colorAndOpacity=<rgba 0 0.1413 0.5089 1>
   }
   slot {class=OverlaySlot, properties {horizontalAlignment=HAlign_Fill, verticalAlignment=VAlign_Bottom}}
  }
  OverlaySpikes {
   widgetClassPath=&Overlay
   parent=Overlay_2
   widget=OverlaySpikes
   properties {renderOpacity=0}
   slot {class=OverlaySlot, properties {horizontalAlignment=HAlign_Fill, verticalAlignment=VAlign_Fill}}
  }
  spike {
   widgetClassPath=&Image
   parent=OverlaySpikes
   widget=spike
   properties {brush {
     bIsDynamicallyLoaded=false
     drawAs=Image
     imageType=NoImage
     mirroring=NoMirror
     resourceName=None
     resourceObject=/FortniteCoreUI/UI/Creative/Materials/M_UI_ActionTriangle_Inst
     tiling=NoTile
     imageSize=<48 48>
     margin=<ltrb 0 0 0 0>
     outlineSettings {
      bUseBrushTransparency=false
      roundingType=HalfHeightRadius
      width=0
      color {colorUseRule=UseColor_Specified, specifiedColor=<rgba 0 0 0 0>}
      cornerRadii=<0 0 0 0>
     }
     tintColor {colorUseRule=UseColor_Specified, specifiedColor=<rgba 1 1 1 1>}
     uVRegion {bIsValid=false, max=<0 0>, min=<0 0>}
    }}
   slot {class=OverlaySlot}
  }
  Image_112 {
   widgetClassPath=&Image
   bIsVariable=true
   parent=OverlaySpikes
   widget=Image_112
   properties {
    brush {
     bIsDynamicallyLoaded=false
     drawAs=Image
     imageType=NoImage
     mirroring=NoMirror
     resourceName=None
     resourceObject=&M_UI_ActionTriangle_Inst
     tiling=NoTile
     imageSize=<28 28>
     margin=<ltrb 0 0 0 0>
     outlineSettings {
      bUseBrushTransparency=false
      roundingType=HalfHeightRadius
      width=0
      color {colorUseRule=UseColor_Specified, specifiedColor=<rgba 0 0 0 0>}
      cornerRadii=<0 0 0 0>
     }
     tintColor {colorUseRule=UseColor_Specified, specifiedColor=<rgba 1 1 1 1>}
     uVRegion {bIsValid=false, max=<0 0>, min=<0 0>}
    }
    renderTransform {angle=0, scale=<-1 1>, shear=<0 0>, translation=<0 22>}
   }
   slot {class=OverlaySlot, properties {horizontalAlignment=HAlign_Right}}
  }
  Image_113 {
   widgetClassPath=&Image
   bIsVariable=true
   parent=OverlaySpikes
   widget=Image_113
   properties {
    brush {
     bIsDynamicallyLoaded=false
     drawAs=Image
     imageType=NoImage
     mirroring=NoMirror
     resourceName=None
     resourceObject=&M_UI_ActionTriangle_Inst
     tiling=NoTile
     imageSize=<48.7241 55.9895>
     margin=<ltrb 0 0 0 0>
     outlineSettings {
      bUseBrushTransparency=false
      roundingType=HalfHeightRadius
      width=0
      color {colorUseRule=UseColor_Specified, specifiedColor=<rgba 0 0 0 0>}
      cornerRadii=<0 0 0 0>
     }
     tintColor {colorUseRule=UseColor_Specified, specifiedColor=<rgba 1 1 1 1>}
     uVRegion {bIsValid=false, max=<0 0>, min=<0 0>}
    }
    renderTransform {angle=0, scale=<-1 1>, shear=<0 0>, translation=<0 0>}
   }
   slot {class=OverlaySlot, properties {horizontalAlignment=HAlign_Right}}
  }
  Image_114 {
   widgetClassPath=&Image
   bIsVariable=true
   parent=OverlaySpikes
   widget=Image_114
   properties {
    brush {
     bIsDynamicallyLoaded=false
     drawAs=Image
     imageType=NoImage
     mirroring=NoMirror
     resourceName=None
     resourceObject=&M_UI_ActionTriangle_Inst
     tiling=NoTile
     imageSize=<64 64>
     margin=<ltrb 0 0 0 0>
     outlineSettings {
      bUseBrushTransparency=false
      roundingType=HalfHeightRadius
      width=0
      color {colorUseRule=UseColor_Specified, specifiedColor=<rgba 0 0 0 0>}
      cornerRadii=<0 0 0 0>
     }
     tintColor {colorUseRule=UseColor_Specified, specifiedColor=<rgba 1 1 1 1>}
     uVRegion {bIsValid=false, max=<0 0>, min=<0 0>}
    }
    renderTransform {angle=0, scale=<1 -1>, shear=<0 0>, translation=<0 0>}
    renderTransformPivot=<0 0.5>
   }
   slot {class=OverlaySlot, properties {verticalAlignment=VAlign_Bottom}}
  }
  Image_115 {
   widgetClassPath=&Image
   bIsVariable=true
   parent=OverlaySpikes
   widget=Image_115
   properties {
    brush {
     bIsDynamicallyLoaded=false
     drawAs=Image
     imageType=NoImage
     mirroring=NoMirror
     resourceName=None
     resourceObject=&M_UI_ActionTriangle_Inst
     tiling=NoTile
     imageSize=<31.3787 32>
     margin=<ltrb 0 0 0 0>
     outlineSettings {
      bUseBrushTransparency=false
      roundingType=HalfHeightRadius
      width=0
      color {colorUseRule=UseColor_Specified, specifiedColor=<rgba 0 0 0 0>}
      cornerRadii=<0 0 0 0>
     }
     tintColor {colorUseRule=UseColor_Specified, specifiedColor=<rgba 1 1 1 1>}
     uVRegion {bIsValid=false, max=<0 0>, min=<0 0>}
    }
    renderTransform {angle=0, scale=<1 -1>, shear=<0 0>, translation=<21 0>}
    renderTransformPivot=<0 0.5>
   }
   slot {class=OverlaySlot, properties {verticalAlignment=VAlign_Bottom}}
  }
  Image_116 {
   widgetClassPath=&Image
   bIsVariable=true
   parent=OverlaySpikes
   widget=Image_116
   properties {
    brush {
     bIsDynamicallyLoaded=false
     drawAs=Image
     imageType=NoImage
     mirroring=NoMirror
     resourceName=None
     resourceObject=&M_UI_ActionTriangle_Inst
     tiling=NoTile
     imageSize=<53.8867 53.4073>
     margin=<ltrb 0 0 0 0>
     outlineSettings {
      bUseBrushTransparency=false
      roundingType=HalfHeightRadius
      width=0
      color {colorUseRule=UseColor_Specified, specifiedColor=<rgba 0 0 0 0>}
      cornerRadii=<0 0 0 0>
     }
     tintColor {colorUseRule=UseColor_Specified, specifiedColor=<rgba 1 1 1 1>}
     uVRegion {bIsValid=false, max=<0 0>, min=<0 0>}
    }
    renderTransform {angle=0, scale=<-1 -1>, shear=<0 0>, translation=<0 0>}
   }
   slot {class=OverlaySlot, properties {horizontalAlignment=HAlign_Right, verticalAlignment=VAlign_Bottom}}
  }
  Image_PanelBorder {
   widgetClassPath=&Image
   parent=Overlay_0
   widget=Image_PanelBorder
   properties {brush {
     bIsDynamicallyLoaded=false
     drawAs=Box
     imageType=NoImage
     mirroring=NoMirror
     resourceName=None
     resourceObject=/Game/UI/Creative/Materials/M_UI_ServerBorder_Selected
     tiling=NoTile
     imageSize=<32 128>
     margin=<ltrb 0 0.17 0 0.17>
     outlineSettings {
      bUseBrushTransparency=false
      roundingType=HalfHeightRadius
      width=0
      color {colorUseRule=UseColor_Specified, specifiedColor=<rgba 0 0 0 0>}
      cornerRadii=<0 0 0 0>
     }
     tintColor {colorUseRule=UseColor_Specified, specifiedColor=<rgba 1 1 1 1>}
     uVRegion {bIsValid=false, max=<0 0>, min=<0 0>}
    }}
   slot {class=OverlaySlot, properties {horizontalAlignment=HAlign_Fill, verticalAlignment=VAlign_Fill}}
  }
  CommonBorder_3 {
   widgetClassPath=&CommonBorder
   parent=Overlay_0
   widget=CommonBorder_3
   properties {padding=<ltrb 24 24 24 16>}
   slot {class=OverlaySlot, properties {horizontalAlignment=HAlign_Fill, verticalAlignment=VAlign_Fill}}
  }
  JoinServerContent {
   widgetClassPath=&Overlay
   bIsVariable=true
   parent=CommonBorder_3
   widget=JoinServerContent
   slot {class=BorderSlot, properties {padding=<ltrb 24 24 24 16>}}
  }
  OwnershipBorder {
   widgetClassPath=&Border
   parent=JoinServerContent
   widget=OwnershipBorder
   properties {brushColor=<rgba 1 1 1 0>}
   slot {class=OverlaySlot, properties {horizontalAlignment=HAlign_Fill, padding=<ltrb 0 0 0 190>}}
  }
  VerticalBox_Ownership {
   widgetClassPath=&VerticalBox
   bInherited=true
   parent=OwnershipBorder
   widget=VerticalBox_Ownership
   properties {visibility=Collapsed}
   slot {class=BorderSlot}
  }
  Border {
   widgetClassPath=&Border
   parent=VerticalBox_Ownership
   widget=Border
   properties {
    visibility=SelfHitTestInvisible
    background {
     bIsDynamicallyLoaded=false
     drawAs=NoDrawType
     imageType=NoImage
     mirroring=NoMirror
     resourceName=None
     resourceObject=null
     tiling=NoTile
     imageSize=<32 32>
     margin=<ltrb 0 0 0 0>
     outlineSettings {
      bUseBrushTransparency=false
      roundingType=HalfHeightRadius
      width=0
      color {colorUseRule=UseColor_Specified, specifiedColor=<rgba 0 0 0 0>}
      cornerRadii=<0 0 0 0>
     }
     tintColor {colorUseRule=UseColor_Specified, specifiedColor=<rgba 1 1 1 1>}
     uVRegion {bIsValid=false, max=<0 0>, min=<0 0>}
    }
    padding=<ltrb 0 0 0 0>
    renderTransform {angle=0, scale=<1 1>, shear=<-8 0>, translation=<0 0>}
   }
   slot {class=VerticalBoxSlot}
  }
  CreatedByText {
   widgetClassPath=/Script/UMG.RichTextBlock
   bIsVariable=true
   parent=Border
   widget=CreatedByText
   properties {
    bOverrideDefaultStyle=true
    justification=Center
    text=CREATED BY
    defaultTextStyleOverride {
     overflowPolicy=Clip
     transformPolicy=None
     colorAndOpacity {colorUseRule=UseColor_Specified, specifiedColor=<rgba 0 1 1 1>}
     font {
      bForceMonospaced=false
      bMaterialIsStencil=false
      fontMaterial=null
      fontObject=&BurbankSmall
      letterSpacing=0
      monospacedWidth=1
      size=16
      skewAmount=0
      typefaceFontName=Bold
      outlineSettings {
       bApplyOutlineToDropShadows=false
       bMiteredCorners=false
       bSeparateFillAlpha=false
       outlineMaterial=null
       outlineSize=2
       outlineColor=<rgba 0 0.0833 0.5 0.175>
      }
     }
     highlightColor {colorUseRule=UseColor_Specified, specifiedColor=<rgba 0 0 0 1>}
     highlightShape {
      bIsDynamicallyLoaded=false
      drawAs=Image
      imageType=NoImage
      mirroring=NoMirror
      resourceName=None
      resourceObject=null
      tiling=NoTile
      imageSize=<32 32>
      margin=<ltrb 0 0 0 0>
      outlineSettings {
       bUseBrushTransparency=false
       roundingType=HalfHeightRadius
       width=0
       color {colorUseRule=UseColor_Specified, specifiedColor=<rgba 0 0 0 0>}
       cornerRadii=<0 0 0 0>
      }
      tintColor {colorUseRule=UseColor_Specified, specifiedColor=<rgba 1 1 1 1>}
      uVRegion {bIsValid=false, max=<0 0>, min=<0 0>}
     }
     selectedBackgroundColor {colorUseRule=UseColor_Foreground, specifiedColor=<rgba 1 0 1 1>}
     shadowColorAndOpacity=<rgba 0 0 0 1>
     shadowOffset=<0 0>
     strikeBrush {
      bIsDynamicallyLoaded=false
      drawAs=Image
      imageType=NoImage
      mirroring=NoMirror
      resourceName=None
      resourceObject=null
      tiling=NoTile
      imageSize=<32 32>
      margin=<ltrb 0 0 0 0>
      outlineSettings {
       bUseBrushTransparency=false
       roundingType=HalfHeightRadius
       width=0
       color {colorUseRule=UseColor_Specified, specifiedColor=<rgba 0 0 0 0>}
       cornerRadii=<0 0 0 0>
      }
      tintColor {colorUseRule=UseColor_Specified, specifiedColor=<rgba 1 1 1 1>}
      uVRegion {bIsValid=false, max=<0 0>, min=<0 0>}
     }
     underlineBrush {
      bIsDynamicallyLoaded=false
      drawAs=Image
      imageType=NoImage
      mirroring=NoMirror
      resourceName=None
      resourceObject=null
      tiling=NoTile
      imageSize=<32 32>
      margin=<ltrb 0 0 0 0>
      outlineSettings {
       bUseBrushTransparency=false
       roundingType=HalfHeightRadius
       width=0
       color {colorUseRule=UseColor_Specified, specifiedColor=<rgba 0 0 0 0>}
       cornerRadii=<0 0 0 0>
      }
      tintColor {colorUseRule=UseColor_Specified, specifiedColor=<rgba 1 1 1 1>}
      uVRegion {bIsValid=false, max=<0 0>, min=<0 0>}
     }
    }
   }
   slot {class=BorderSlot, properties {padding=<ltrb 0 0 0 0>}}
  }
  Border_1 {
   widgetClassPath=&Border
   parent=VerticalBox_Ownership
   widget=Border_1
   properties {brushColor=<rgba 1 1 1 0>, padding=<ltrb 4 0 4 2>}
   slot {class=VerticalBoxSlot}
  }
  Text_OwnerName {
   widgetClassPath=&CommonTextBlock
   bInherited=true
   bIsVariable=true
   parent=Border_1
   widget=Text_OwnerName
   properties {
    justification=Center
    lineHeightPercentage=0.8
    text=WhiteWalker
    font {
     bForceMonospaced=false
     bMaterialIsStencil=false
     fontMaterial=null
     fontObject=&BurbankSmall
     letterSpacing=0
     monospacedWidth=1
     size=24
     skewAmount=0
     typefaceFontName=Black
     outlineSettings {
      bApplyOutlineToDropShadows=false
      bMiteredCorners=false
      bSeparateFillAlpha=false
      outlineMaterial=null
      outlineSize=2
      outlineColor=<rgba 0 0.001 0.0125 0.2>
     }
    }
    margin=<ltrb 0 -8 0 0>
    shadowOffset=<0 0>
   }
   slot {class=BorderSlot, properties {padding=<ltrb 4 0 4 2>}}
  }
  CommonBorder_ServerDetails {
   widgetClassPath=&CommonBorder
   bIsVariable=true
   parent=JoinServerContent
   widget=CommonBorder_ServerDetails
   properties {padding=<ltrb 0 0 0 0>}
   slot {class=OverlaySlot, properties {horizontalAlignment=HAlign_Fill, verticalAlignment=VAlign_Bottom, padding=<ltrb 0 10 0 4>}}
  }
  VerticalBox_2 {
   widgetClassPath=&VerticalBox
   parent=CommonBorder_ServerDetails
   widget=VerticalBox_2
   properties {renderTransformPivot=<0.5 0.9>}
   slot {class=BorderSlot, properties {padding=<ltrb 0 0 0 0>}}
  }
  CommonBorderInspectTitle {
   widgetClassPath=&CommonBorder
   parent=VerticalBox_2
   widget=CommonBorderInspectTitle
   properties {clipping=ClipToBounds, renderOpacity=0, desiredSizeScale=<1 0>, padding=<ltrb 0 -10 0 0>}
   slot {class=VerticalBoxSlot}
  }
  SizeBox_3 {
   widgetClassPath=&SizeBox
   parent=CommonBorderInspectTitle
   widget=SizeBox_3
   slot {class=BorderSlot, properties {padding=<ltrb 0 -10 0 0>}}
  }
  ScaleBox_0 {
   widgetClassPath=&ScaleBox
   parent=SizeBox_3
   widget=ScaleBox_0
   properties {stretch=ScaleToFitX, stretchDirection=DownOnly}
   slot {class=SizeBoxSlot}
  }
  FriendsHereTextBlock {
   widgetClassPath=&CommonTextBlock
   parent=ScaleBox_0
   widget=FriendsHereTextBlock
   properties {
    justification=Center
    lineHeightPercentage=0.8
    text=FRIENDS HERE
    colorAndOpacity {colorUseRule=UseColor_Specified, specifiedColor=<rgba 1 1 1 0.66>}
    font {
     bForceMonospaced=false
     bMaterialIsStencil=false
     fontMaterial=null
     fontObject=&Burbank
     letterSpacing=0
     monospacedWidth=1
     size=40
     skewAmount=0
     typefaceFontName=Burbank-Black
     outlineSettings {
      bApplyOutlineToDropShadows=false
      bMiteredCorners=false
      bSeparateFillAlpha=false
      outlineMaterial=null
      outlineSize=0
      outlineColor=<rgba 0 0 0 1>
     }
    }
   }
   slot {class=ScaleBoxSlot}
  }
  CommonBorderDefaultTitle {
   widgetClassPath=&CommonBorder
   parent=VerticalBox_2
   widget=CommonBorderDefaultTitle
   properties {clipping=ClipToBounds, padding=<ltrb 0 0 0 0>}
   slot {class=VerticalBoxSlot}
  }
  VerticalBox_0 {
   widgetClassPath=&VerticalBox
   parent=CommonBorderDefaultTitle
   widget=VerticalBox_0
   slot {class=BorderSlot, properties {padding=<ltrb 0 0 0 0>}}
  }
  ScaleBox_2 {
   widgetClassPath=&ScaleBox
   parent=VerticalBox_0
   widget=ScaleBox_2
   properties {stretchDirection=DownOnly}
   slot {class=VerticalBoxSlot, properties {padding=<ltrb 0 0 0 12>}}
  }
  Text_FriendsPlaying {
   widgetClassPath=&CommonTextBlock
   bInherited=true
   bIsVariable=true
   parent=ScaleBox_2
   widget=Text_FriendsPlaying
   properties {
    justification=Center
    lineHeightPercentage=0.8
    text="JONSNOW\r\nARYASTARK\r\nCERSEILANNISTER"
    font {
     bForceMonospaced=false
     bMaterialIsStencil=false
     fontMaterial=null
     fontObject=&BurbankSmall
     letterSpacing=0
     monospacedWidth=1
     size=24
     skewAmount=0
     typefaceFontName=Black
     outlineSettings {
      bApplyOutlineToDropShadows=false
      bMiteredCorners=false
      bSeparateFillAlpha=false
      outlineMaterial=null
      outlineSize=2
      outlineColor=<rgba 0 0.001 0.0125 0.2>
     }
    }
    shadowOffset=<0 0>
   }
   slot {class=ScaleBoxSlot}
  }
  Border_0 {
   widgetClassPath=&Border
   parent=VerticalBox_0
   widget=Border_0
   properties {
    visibility=SelfHitTestInvisible
    background {
     bIsDynamicallyLoaded=false
     drawAs=NoDrawType
     imageType=NoImage
     mirroring=NoMirror
     resourceName=None
     resourceObject=null
     tiling=NoTile
     imageSize=<32 32>
     margin=<ltrb 0 0 0 0>
     outlineSettings {
      bUseBrushTransparency=false
      roundingType=HalfHeightRadius
      width=0
      color {colorUseRule=UseColor_Specified, specifiedColor=<rgba 0 0 0 0>}
      cornerRadii=<0 0 0 0>
     }
     tintColor {colorUseRule=UseColor_Specified, specifiedColor=<rgba 1 1 1 1>}
     uVRegion {bIsValid=false, max=<0 0>, min=<0 0>}
    }
    padding=<ltrb 0 0 0 0>
    renderTransform {angle=0, scale=<1 1>, shear=<-8 0>, translation=<0 0>}
   }
   slot {class=VerticalBoxSlot}
  }
  RichText_RemainingFriends {
   widgetClassPath=&RichTextBlock
   bInherited=true
   bIsVariable=true
   parent=Border_0
   widget=RichText_RemainingFriends
   properties {
    bOverrideDefaultStyle=true
    justification=Center
    text=AND 7 MORE FRIENDS
    textStyleSet=/FortniteCoreUI/UI/Creative/CreativeFrontendStyles
    defaultTextStyleOverride {
     overflowPolicy=Clip
     transformPolicy=None
     colorAndOpacity {colorUseRule=UseColor_Specified, specifiedColor=<rgba 0 1 1 1>}
     font {
      bForceMonospaced=false
      bMaterialIsStencil=false
      fontMaterial=null
      fontObject=&Burbank
      letterSpacing=0
      monospacedWidth=1
      size=18
      skewAmount=0
      typefaceFontName=BurbankRegular-Black
      outlineSettings {
       bApplyOutlineToDropShadows=false
       bMiteredCorners=false
       bSeparateFillAlpha=false
       outlineMaterial=null
       outlineSize=2
       outlineColor=<rgba 0 0.0833 0.5 0.175>
      }
     }
     highlightColor {colorUseRule=UseColor_Specified, specifiedColor=<rgba 0 0 0 1>}
     highlightShape {
      bIsDynamicallyLoaded=false
      drawAs=Image
      imageType=NoImage
      mirroring=NoMirror
      resourceName=None
      resourceObject=null
      tiling=NoTile
      imageSize=<32 32>
      margin=<ltrb 0 0 0 0>
      outlineSettings {
       bUseBrushTransparency=false
       roundingType=HalfHeightRadius
       width=0
       color {colorUseRule=UseColor_Specified, specifiedColor=<rgba 0 0 0 0>}
       cornerRadii=<0 0 0 0>
      }
      tintColor {colorUseRule=UseColor_Specified, specifiedColor=<rgba 1 1 1 1>}
      uVRegion {bIsValid=false, max=<0 0>, min=<0 0>}
     }
     selectedBackgroundColor {colorUseRule=UseColor_Foreground, specifiedColor=<rgba 1 0 1 1>}
     shadowColorAndOpacity=<rgba 0 0 0 1>
     shadowOffset=<0 0>
     strikeBrush {
      bIsDynamicallyLoaded=false
      drawAs=Image
      imageType=NoImage
      mirroring=NoMirror
      resourceName=None
      resourceObject=null
      tiling=NoTile
      imageSize=<32 32>
      margin=<ltrb 0 0 0 0>
      outlineSettings {
       bUseBrushTransparency=false
       roundingType=HalfHeightRadius
       width=0
       color {colorUseRule=UseColor_Specified, specifiedColor=<rgba 0 0 0 0>}
       cornerRadii=<0 0 0 0>
      }
      tintColor {colorUseRule=UseColor_Specified, specifiedColor=<rgba 1 1 1 1>}
      uVRegion {bIsValid=false, max=<0 0>, min=<0 0>}
     }
     underlineBrush {
      bIsDynamicallyLoaded=false
      drawAs=Image
      imageType=NoImage
      mirroring=NoMirror
      resourceName=None
      resourceObject=null
      tiling=NoTile
      imageSize=<32 32>
      margin=<ltrb 0 0 0 0>
      outlineSettings {
       bUseBrushTransparency=false
       roundingType=HalfHeightRadius
       width=0
       color {colorUseRule=UseColor_Specified, specifiedColor=<rgba 0 0 0 0>}
       cornerRadii=<0 0 0 0>
      }
      tintColor {colorUseRule=UseColor_Specified, specifiedColor=<rgba 1 1 1 1>}
      uVRegion {bIsValid=false, max=<0 0>, min=<0 0>}
     }
    }
   }
   slot {class=BorderSlot, properties {padding=<ltrb 0 0 0 0>}}
  }
  SizeBox_0 {
   widgetClassPath=&SizeBox
   parent=VerticalBox_0
   widget=SizeBox_0
   properties {bOverride_HeightOverride=true, heightOverride=64}
   slot {class=VerticalBoxSlot, properties {verticalAlignment=VAlign_Center, size {sizeRule=Fill, value=1}}}
  }
  Launch_Overlay {
   widgetClassPath=&ScaleBox
   bIsVariable=true
   parent=SizeBox_0
   widget=Launch_Overlay
   properties {visibility=Collapsed}
   slot {class=SizeBoxSlot}
  }
  Button_Launch {
   widgetClassPath=/FortniteCoreUI/Frontend/Showdown/PrimarySkewButton
   bInherited=true
   bIsVariable=true
   parent=Launch_Overlay
   widget=Button_Launch
   properties {
    buttonNameText=LAUNCH
    onButtonBaseClicked=(&self:WidgetTree.Button_Launch.HandleOnButtonClicked)
    onButtonBaseDoubleClicked=(&self:WidgetTree.Button_Launch.HandleOnButtonDoubleClicked)
    onButtonBaseHovered=(&self:WidgetTree.Button_Launch.HandleOnButtonHovered)
    onButtonBaseUnhovered=(&self:WidgetTree.Button_Launch.HandleOnButtonUnhovered)
    onSelectedChangedBase=(&self:WidgetTree.Button_Launch.HandleOnSelectedChanged)
    triggeringInputAction {dataTable=&InputActionDataTable, rowName=Confirm}
   }
   slot {class=ScaleBoxSlot}
  }
  OverlayFriendCount {
   widgetClassPath=&Overlay
   parent=VerticalBox_0
   widget=OverlayFriendCount
   slot {class=VerticalBoxSlot, properties {padding=<ltrb 4 12 4 0>}}
  }
  Button_ShowFriends {
   widgetClassPath=/Game/Athena/UI/Matchmaking/SimpleCommonButtonWithInput
   bIsVariable=true
   parent=OverlayFriendCount
   widget=Button_ShowFriends
   properties {
    onButtonBaseClicked=(&self:WidgetTree.Button_ShowFriends.HandleOnButtonClicked)
    onButtonBaseDoubleClicked=(&self:WidgetTree.Button_ShowFriends.HandleOnButtonDoubleClicked)
    onButtonBaseHovered=(&self:WidgetTree.Button_ShowFriends.HandleOnButtonHovered)
    onButtonBaseUnhovered=(&self:WidgetTree.Button_ShowFriends.HandleOnButtonUnhovered)
    onSelectedChangedBase=(&self:WidgetTree.Button_ShowFriends.HandleOnSelectedChanged)
    style=&ButtonStyle-Serverbrowser_C
    visibility=Collapsed
    triggeringInputAction {dataTable=/FortUILibrary/Input/CreativeInputActionDataTable, rowName=ServerToggleOptions}
   }
   slot {class=OverlaySlot, properties {horizontalAlignment=HAlign_Fill, verticalAlignment=VAlign_Fill, padding=<ltrb 34 0 0 0>}}
  }
  HorizontalBox_0 {
   widgetClassPath=&HorizontalBox
   parent=OverlayFriendCount
   widget=HorizontalBox_0
   slot {class=OverlaySlot, properties {horizontalAlignment=HAlign_Center, verticalAlignment=VAlign_Center, padding=<ltrb 0 8 0 8>}}
  }
  Image_28 {
   widgetClassPath=&Image
   bIsVariable=true
   parent=HorizontalBox_0
   widget=Image_28
   properties {
    visibility=SelfHitTestInvisible
    brush {
     bIsDynamicallyLoaded=false
     drawAs=Image
     imageType=NoImage
     mirroring=NoMirror
     resourceName=None
     resourceObject=/FortniteCoreUI/UI/Icons/Icon-Social-M
     tiling=NoTile
     imageSize=<30 30>
     margin=<ltrb 0 0 0 0>
     outlineSettings {
      bUseBrushTransparency=false
      roundingType=HalfHeightRadius
      width=0
      color {colorUseRule=UseColor_Specified, specifiedColor=<rgba 0 0 0 0>}
      cornerRadii=<0 0 0 0>
     }
     tintColor {colorUseRule=UseColor_Specified, specifiedColor=<rgba 1 1 1 1>}
     uVRegion {bIsValid=false, max=<0 0>, min=<0 0>}
    }
    renderTransform {angle=0, scale=<1 1>, shear=<0 0>, translation=<0 1>}
   }
   slot {class=HorizontalBoxSlot, properties {horizontalAlignment=HAlign_Center, verticalAlignment=VAlign_Center, padding=<ltrb 0 0 8 0>}}
  }
  Text_PlayerCount {
   widgetClassPath=&CommonTextBlock
   bInherited=true
   bIsVariable=true
   parent=HorizontalBox_0
   widget=Text_PlayerCount
   properties {
    justification=Center
    text=00/00
    font {
     bForceMonospaced=false
     bMaterialIsStencil=false
     fontMaterial=null
     fontObject=&Burbank
     letterSpacing=0
     monospacedWidth=1
     size=24
     skewAmount=0
     typefaceFontName=BurbankRegular-Black
     outlineSettings {
      bApplyOutlineToDropShadows=false
      bMiteredCorners=false
      bSeparateFillAlpha=false
      outlineMaterial=null
      outlineSize=0
      outlineColor=<rgba 0 0 0 1>
     }
    }
    margin=<ltrb 0 0 0 -5>
    renderTransform {angle=0, scale=<1 1>, shear=<0 0>, translation=<0 -2>}
    shadowOffset=<0 0>
   }
   slot {class=HorizontalBoxSlot, properties {horizontalAlignment=HAlign_Center, verticalAlignment=VAlign_Center}}
  }
  WarningOverlay {
   widgetClassPath=&Overlay
   bIsVariable=true
   parent=VerticalBox_0
   widget=WarningOverlay
   slot {class=VerticalBoxSlot, properties {padding=<ltrb 0 5 0 4>}}
  }
  Text_ServerWarning {
   widgetClassPath=&CommonTextBlock
   bInherited=true
   bIsVariable=true
   parent=WarningOverlay
   widget=Text_ServerWarning
   properties {
    autoWrapText=true
    justification=Center
    text=WARNING TEXT
    visibility=Collapsed
    colorAndOpacity {colorUseRule=UseColor_Specified, specifiedColor=<rgba 1 0.9301 0 1>}
    font {
     bForceMonospaced=false
     bMaterialIsStencil=false
     fontMaterial=null
     fontObject=&BurbankSmall
     letterSpacing=0
     monospacedWidth=1
     size=12
     skewAmount=0
     typefaceFontName=Black
     outlineSettings {
      bApplyOutlineToDropShadows=false
      bMiteredCorners=false
      bSeparateFillAlpha=false
      outlineMaterial=null
      outlineSize=0
      outlineColor=<rgba 0 0 0 1>
     }
    }
    shadowOffset=<0 0>
   }
   slot {class=OverlaySlot, properties {horizontalAlignment=HAlign_Fill, verticalAlignment=VAlign_Fill}}
  }
  CommonBorder_1 {
   widgetClassPath=&CommonBorder
   parent=VerticalBox_2
   widget=CommonBorder_1
   properties {clipping=ClipToBounds, verticalAlignment=VAlign_Top, desiredSizeScale=<1 0>, padding=<ltrb 0 10 0 0>}
   slot {class=VerticalBoxSlot, properties {size {sizeRule=Fill, value=1}}}
  }
  SizeBox_2 {
   widgetClassPath=&SizeBox
   parent=CommonBorder_1
   widget=SizeBox_2
   properties {bOverride_HeightOverride=true, bOverride_MinDesiredHeight=true, heightOverride=413.0634, minDesiredHeight=460, widthOverride=450}
   slot {class=BorderSlot, properties {verticalAlignment=VAlign_Top, padding=<ltrb 0 10 0 0>}}
  }
  ScrollBox_2 {
   widgetClassPath=&ScrollBox
   parent=SizeBox_2
   widget=ScrollBox_2
   properties {
    scrollBarVisibility=Collapsed
    widgetStyle {
     barThickness=9
     bottomShadowBrush {
      bIsDynamicallyLoaded=false
      drawAs=NoDrawType
      imageType=FullColor
      mirroring=NoMirror
      resourceName=../../../Engine/Content/Slate/Common/ScrollBoxShadowTop.png
      resourceObject=null
      tiling=NoTile
      imageSize=<16 8>
      margin=<ltrb 0.5 1 0.5 0>
      outlineSettings {
       bUseBrushTransparency=false
       roundingType=HalfHeightRadius
       width=0
       color {colorUseRule=UseColor_Specified, specifiedColor=<rgba 0 0 0 0>}
       cornerRadii=<0 0 0 0>
      }
      tintColor {colorUseRule=UseColor_Specified, specifiedColor=<rgba 1 1 1 1>}
      uVRegion {bIsValid=false, max=<0 0>, min=<0 0>}
     }
     horizontalScrolledContentPadding=<ltrb 0 0 1 0>
     leftShadowBrush {
      bIsDynamicallyLoaded=false
      drawAs=NoDrawType
      imageType=FullColor
      mirroring=NoMirror
      resourceName=../../../Engine/Content/Slate/Common/ScrollBoxShadowTop.png
      resourceObject=null
      tiling=NoTile
      imageSize=<16 8>
      margin=<ltrb 0.5 1 0.5 0>
      outlineSettings {
       bUseBrushTransparency=false
       roundingType=HalfHeightRadius
       width=0
       color {colorUseRule=UseColor_Specified, specifiedColor=<rgba 0 0 0 0>}
       cornerRadii=<0 0 0 0>
      }
      tintColor {colorUseRule=UseColor_Specified, specifiedColor=<rgba 1 1 1 1>}
      uVRegion {bIsValid=false, max=<0 0>, min=<0 0>}
     }
     rightShadowBrush {
      bIsDynamicallyLoaded=false
      drawAs=NoDrawType
      imageType=FullColor
      mirroring=NoMirror
      resourceName=../../../Engine/Content/Slate/Common/ScrollBoxShadowTop.png
      resourceObject=null
      tiling=NoTile
      imageSize=<16 8>
      margin=<ltrb 0.5 1 0.5 0>
      outlineSettings {
       bUseBrushTransparency=false
       roundingType=HalfHeightRadius
       width=0
       color {colorUseRule=UseColor_Specified, specifiedColor=<rgba 0 0 0 0>}
       cornerRadii=<0 0 0 0>
      }
      tintColor {colorUseRule=UseColor_Specified, specifiedColor=<rgba 1 1 1 1>}
      uVRegion {bIsValid=false, max=<0 0>, min=<0 0>}
     }
     topShadowBrush {
      bIsDynamicallyLoaded=false
      drawAs=NoDrawType
      imageType=FullColor
      mirroring=NoMirror
      resourceName=../../../Engine/Content/Slate/Common/ScrollBoxShadowTop.png
      resourceObject=null
      tiling=NoTile
      imageSize=<16 8>
      margin=<ltrb 0.5 1 0.5 0>
      outlineSettings {
       bUseBrushTransparency=false
       roundingType=HalfHeightRadius
       width=0
       color {colorUseRule=UseColor_Specified, specifiedColor=<rgba 0 0 0 0>}
       cornerRadii=<0 0 0 0>
      }
      tintColor {colorUseRule=UseColor_Specified, specifiedColor=<rgba 1 1 1 1>}
      uVRegion {bIsValid=false, max=<0 0>, min=<0 0>}
     }
     verticalScrolledContentPadding=<ltrb 0 0 0 1>
    }
   }
   slot {class=SizeBoxSlot}
  }
  EntryBox_ServerFriendList {
   widgetClassPath=&DynamicEntryBox
   bInherited=true
   bIsVariable=true
   parent=ScrollBox_2
   widget=EntryBox_ServerFriendList
   properties {
    entryBoxType=Vertical
    entryHorizontalAlignment=HAlign_Center
    entryWidgetClass=&/ServerBrowserFriendEntry.ServerBrowserFriendEntry_C
    numDesignerPreviewEntries=16
    entrySpacing=<0 8>
   }
   slot {class=ScrollBoxSlot}
  }
  CommonBorderBackButton {
   widgetClassPath=&CommonBorder
   parent=VerticalBox_2
   widget=CommonBorderBackButton
   properties {clipping=ClipToBounds, renderOpacity=0, desiredSizeScale=<1 0>, padding=<ltrb 0 0 0 0>}
   slot {class=VerticalBoxSlot}
  }
  OverlayBackButton {
   widgetClassPath=&Overlay
   parent=CommonBorderBackButton
   widget=OverlayBackButton
   slot {class=BorderSlot, properties {padding=<ltrb 0 0 0 0>}}
  }
  Button_CloseFriends {
   widgetClassPath=&SimpleCommonButtonWithInput
   bIsVariable=true
   parent=OverlayBackButton
   widget=Button_CloseFriends
   properties {
    onButtonBaseClicked=(&self:WidgetTree.Button_CloseFriends.HandleOnButtonClicked)
    onButtonBaseDoubleClicked=(&self:WidgetTree.Button_CloseFriends.HandleOnButtonDoubleClicked)
    onButtonBaseHovered=(&self:WidgetTree.Button_CloseFriends.HandleOnButtonHovered)
    onButtonBaseUnhovered=(&self:WidgetTree.Button_CloseFriends.HandleOnButtonUnhovered)
    onSelectedChangedBase=(&self:WidgetTree.Button_CloseFriends.HandleOnSelectedChanged)
    style=/FortUILibrary/Foundation/Buttons/ButtonStyles/ButtonStyle-TextOnlyBase_Empty.ButtonStyle-TextOnlyBase_Empty_C
    visibility=Visible
    triggeringInputAction {dataTable=&CreativeInputActionDataTable, rowName=ServerToggleOptions}
   }
   slot {class=OverlaySlot, properties {horizontalAlignment=HAlign_Fill, verticalAlignment=VAlign_Fill, padding=<ltrb 50 0 50 0>}}
  }
  HorizontalBox_1 {
   widgetClassPath=&HorizontalBox
   parent=OverlayBackButton
   widget=HorizontalBox_1
   properties {visibility=HitTestInvisible}
   slot {class=OverlaySlot, properties {horizontalAlignment=HAlign_Center}}
  }
  CommonTextBlock_1 {
   widgetClassPath=&CommonTextBlock
   parent=HorizontalBox_1
   widget=CommonTextBlock_1
   properties {
    justification=Center
    text=Back
    font {
     bForceMonospaced=false
     bMaterialIsStencil=false
     fontMaterial=null
     fontObject=&Burbank
     letterSpacing=0
     monospacedWidth=1
     size=32
     skewAmount=0
     typefaceFontName=Burbank-Black
     outlineSettings {
      bApplyOutlineToDropShadows=false
      bMiteredCorners=false
      bSeparateFillAlpha=false
      outlineMaterial=null
      outlineSize=0
      outlineColor=<rgba 0 0 0 1>
     }
    }
    shadowOffset=<0 0>
   }
   slot {class=HorizontalBoxSlot, properties {horizontalAlignment=HAlign_Center, verticalAlignment=VAlign_Center}}
  }
  InputActionWidget {widgetClassPath=&CommonActionWidget, bInherited=true, widget=null}
 ]
}
