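# Server browser action tile, built on FortPlaylistSelectionActionTile, for launching a new playlist server, with "Create a BattleLab" as default title.
# Selection reveals description, warning and launch elements and plays Selected_1 to widen and scale the tile; deselection reverses this.
# Hover animation is skipped on touch, when selected, or while Selected_1 plays.
# The privacy button flips a local bool and label; launch-enable changes swap LAUNCH/DISABLED text; random-background logic is unwired.
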WidgetBlueprint /FortniteCoreUI/Frontend/ActivityBrowser/ServerBrowserPlaylistActionTile
animations [
 Selected {
  binding_count=5
  length_seconds=1
  object=&self:Selected
  sequence {bindings [
    CommonBorder_0 {
     id=#1
     possessed_class=CommonBorder
     tracks [
      {
       type=MovieSceneFloatVectorTrack
       display_name=Desired Size Scale
       sections [{
         type=MovieSceneFloatVectorSection
         channels [
          Vector.X {type=MovieSceneScriptingFloatChannel}
          Vector.Y {type=MovieSceneScriptingFloatChannel, keys [[0,0],[20,1]]}
         ]
        }]
      }
      {
       type=MovieSceneFloatTrack
       display_name=Render Opacity
       sections [{type=MovieSceneFloatSection, channels [None {type=MovieSceneScriptingFloatChannel, keys [[0,0],[1.994,1]]}]}]
      }
     ]
    }
    SizeBox_Root {
     id=#2
     possessed_class=SizeBox
     tracks [
      {
       type=MovieSceneFloatTrack
       display_name=Width Override
       sections [{type=MovieSceneFloatSection, channels [None {type=MovieSceneScriptingFloatChannel, keys [[0,315.4871],[20,435]]}]}]
      }
      {
       type=MovieScene2DTransformTrack
       sections [{
         type=MovieScene2DTransformSection
         end_frame=20
         start_frame=0
         channels [
          Translation.X {type=MovieSceneScriptingFloatChannel}
          Translation.Y {type=MovieSceneScriptingFloatChannel}
          Angle {type=MovieSceneScriptingFloatChannel}
          Scale.X {type=MovieSceneScriptingFloatChannel, keys [[0,0.8],[15,1.0449],[19,0.9924],[20,1]]}
          Scale.Y {type=MovieSceneScriptingFloatChannel, keys [[0,0.8],[15,1.0449],[19,0.9924],[20,1]]}
          Shear.X {type=MovieSceneScriptingFloatChannel}
          Shear.Y {type=MovieSceneScriptingFloatChannel}
         ]
        }]
      }
     ]
    }
    BorderSlot_0 {
     id=#3
     parent_id=#2
     possessed_class=BorderSlot
     tracks [{
       type=MovieSceneMarginTrack
       display_name=Padding
       sections [{
         type=MovieSceneMarginSection
         end_frame=20
         start_frame=0
         channels [
          Left {type=MovieSceneScriptingFloatChannel, keys [[0,-26.0614],[20,13.6259]]}
          Top {type=MovieSceneScriptingFloatChannel, const [0,2]}
          Right {type=MovieSceneScriptingFloatChannel, keys [[0,-26.0438],[20,12]]}
          Bottom {type=MovieSceneScriptingFloatChannel, const [0,2]}
         ]
        }]
      }]
    }
    Image_ColorFill {
     id=#4
     possessed_class=Image
     tracks [{
       type=MovieSceneFloatTrack
       display_name=Render Opacity
       sections [{type=MovieSceneFloatSection, end_frame=20, start_frame=0, channels [None {type=MovieSceneScriptingFloatChannel, keys [[0,0.45],[20,0]]}]}]
      }]
    }
    Image_PanelBorder {
     id=#5
     possessed_class=Image
     tracks [{
       type=MovieSceneWidgetMaterialTrack
       sections [{
         type=MovieSceneParameterSection
         channels [
          BottomEdge_1 {type=MovieSceneScriptingFloatChannel, keys [[0,-0.0222],[20,0.0583]]}
          LeftEdge_1 {type=MovieSceneScriptingFloatChannel, keys [[0,-0.009],[20,0.024]]}
          RightEdge_1 {type=MovieSceneScriptingFloatChannel, keys [[0,0],[20,0.027]]}
          TopEdge_1 {type=MovieSceneScriptingFloatChannel, keys [[0,0],[20,0.081]]}
         ]
        }]
      }]
    }
   ]}
 }
 ViewPlayers {
  binding_count=10
  length_seconds=0.3676
  object=&self:ViewPlayers
  sequence {
   bindings [
    BackgroundBlur_0 {
     id=#6
     possessed_class=BackgroundBlur
     tracks [{
       type=MovieSceneFloatTrack
       display_name=Blur Strength
       sections [{type=MovieSceneFloatSection, channels [None {type=MovieSceneScriptingFloatChannel, keys [[0,0],[44.112,10]]}]}]
      }]
    }
    Image_112 {
     id=#7
     possessed_class=Image
     tracks [{
       type=MovieScene2DTransformTrack
       sections [{
         type=MovieScene2DTransformSection
         end_frame=44
         start_frame=0
         channels [
          Translation.X {type=MovieSceneScriptingFloatChannel}
          Translation.Y {type=MovieSceneScriptingFloatChannel}
          Angle {type=MovieSceneScriptingFloatChannel}
          Scale.X {type=MovieSceneScriptingFloatChannel, keys [[0,0],[44.094,-1]]}
          Scale.Y {type=MovieSceneScriptingFloatChannel, keys [[0,0],[44.094,1]]}
          Shear.X {type=MovieSceneScriptingFloatChannel}
          Shear.Y {type=MovieSceneScriptingFloatChannel}
         ]
        }]
      }]
    }
    Image_116 {
     id=#8
     possessed_class=Image
     tracks [{
       type=MovieScene2DTransformTrack
       sections [{
         type=MovieScene2DTransformSection
         end_frame=44
         start_frame=0
         channels [
          Translation.X {type=MovieSceneScriptingFloatChannel}
          Translation.Y {type=MovieSceneScriptingFloatChannel}
          Angle {type=MovieSceneScriptingFloatChannel}
          Scale.X {type=MovieSceneScriptingFloatChannel, keys [[0,0],[44.098,-1]]}
          Scale.Y {type=MovieSceneScriptingFloatChannel, keys [[0,0],[44.098,-1]]}
          Shear.X {type=MovieSceneScriptingFloatChannel}
          Shear.Y {type=MovieSceneScriptingFloatChannel}
         ]
        }]
      }]
    }
    Image_ColorFill {
     id=#9
     possessed_class=Image
     tracks [{
       type=MovieSceneFloatTrack
       display_name=Render Opacity
       sections [{type=MovieSceneFloatSection, channels [None {type=MovieSceneScriptingFloatChannel, keys [[0,0],[44.112,0.9]]}]}]
      }]
    }
    OverlaySpikes {
     id=#10
     possessed_class=Overlay
     tracks [{
       type=MovieSceneFloatTrack
       display_name=Render Opacity
       sections [{type=MovieSceneFloatSection, channels [None {type=MovieSceneScriptingFloatChannel, keys [[0,0],[44.102,1]]}]}]
      }]
    }
    Image_114 {
     id=#11
     possessed_class=Image
     tracks [{
       type=MovieScene2DTransformTrack
       sections [{
         type=MovieScene2DTransformSection
         end_frame=44
         start_frame=0
         channels [
          Translation.X {type=MovieSceneScriptingFloatChannel}
          Translation.Y {type=MovieSceneScriptingFloatChannel}
          Angle {type=MovieSceneScriptingFloatChannel}
          Scale.X {type=MovieSceneScriptingFloatChannel, keys [[0,0],[44.106,1]]}
          Scale.Y {type=MovieSceneScriptingFloatChannel, keys [[0,0],[44.106,-1]]}
          Shear.X {type=MovieSceneScriptingFloatChannel}
          Shear.Y {type=MovieSceneScriptingFloatChannel}
         ]
        }]
      }]
    }
    spike {
     id=#12
     possessed_class=Image
     tracks [{
       type=MovieScene2DTransformTrack
       sections [{
         type=MovieScene2DTransformSection
         end_frame=44
         start_frame=0
         channels [
          Translation.X {type=MovieSceneScriptingFloatChannel}
          Translation.Y {type=MovieSceneScriptingFloatChannel}
          Angle {type=MovieSceneScriptingFloatChannel}
          Scale.X {type=MovieSceneScriptingFloatChannel, keys [[0,0],[44.106,1]]}
          Scale.Y {type=MovieSceneScriptingFloatChannel, keys [[0,0],[44.106,1]]}
          Shear.X {type=MovieSceneScriptingFloatChannel}
          Shear.Y {type=MovieSceneScriptingFloatChannel}
         ]
        }]
      }]
    }
    Image_115 {
     id=#13
     possessed_class=Image
     tracks [{
       type=MovieScene2DTransformTrack
       sections [{
         type=MovieScene2DTransformSection
         end_frame=44
         start_frame=0
         channels [
          Translation.X {type=MovieSceneScriptingFloatChannel}
          Translation.Y {type=MovieSceneScriptingFloatChannel}
          Angle {type=MovieSceneScriptingFloatChannel}
          Scale.X {type=MovieSceneScriptingFloatChannel, keys [[0,0],[44.106,1]]}
          Scale.Y {type=MovieSceneScriptingFloatChannel, keys [[0,0],[44.106,-1]]}
          Shear.X {type=MovieSceneScriptingFloatChannel}
          Shear.Y {type=MovieSceneScriptingFloatChannel}
         ]
        }]
      }]
    }
    Image_Gradient {
     id=#14
     possessed_class=Image
     tracks [{
       type=MovieSceneFloatTrack
       display_name=Render Opacity
       sections [{type=MovieSceneFloatSection, channels [None {type=MovieSceneScriptingFloatChannel, keys [[0,1],[44.106,0.5]]}]}]
      }]
    }
    Image_113 {
     id=#15
     possessed_class=Image
     tracks [{
       type=MovieScene2DTransformTrack
       sections [{
         type=MovieScene2DTransformSection
         end_frame=44
         start_frame=0
         channels [
          Translation.X {type=MovieSceneScriptingFloatChannel}
          Translation.Y {type=MovieSceneScriptingFloatChannel}
          Angle {type=MovieSceneScriptingFloatChannel}
          Scale.X {type=MovieSceneScriptingFloatChannel, keys [[0,0],[44.106,-1]]}
          Scale.Y {type=MovieSceneScriptingFloatChannel, keys [[0,0],[44.106,1]]}
          Shear.X {type=MovieSceneScriptingFloatChannel}
          Shear.Y {type=MovieSceneScriptingFloatChannel}
         ]
        }]
      }]
    }
   ]
   display_rate {numerator=120}
  }
 }
 HoverToggle {
  binding_count=1
  length_seconds=0.303
  object=&self:HoverToggle
  sequence {bindings [Image_DetailsButtonBG {
     id=#16
     possessed_class=Image
     tracks [{
       type=MovieSceneColorTrack
       display_name=Color and Opacity
       sections [{
         type=MovieSceneColorSection
         end_frame=6
         start_frame=0
         channels [
          Color.R {type=MovieSceneScriptingFloatChannel, const [1,2]}
          Color.G {type=MovieSceneScriptingFloatChannel, const [1,2]}
          Color.B {type=MovieSceneScriptingFloatChannel, keys [[0,1],[6.0593,0.1]]}
          Color.A {type=MovieSceneScriptingFloatChannel, const [1,2]}
         ]
        }]
      }]
    }]}
 }
 PrivacyToggle {
  binding_count=1
  length_seconds=0.303
  object=&self:PrivacyToggle
  sequence {bindings [ScaleBox_DetailsText {
     id=#17
     possessed_class=ScaleBox
     tracks [{
       type=MovieScene2DTransformTrack
       sections [{
         type=MovieScene2DTransformSection
         end_frame=6
         start_frame=0
         channels [
          Translation.X {type=MovieSceneScriptingFloatChannel}
          Translation.Y {type=MovieSceneScriptingFloatChannel}
          Angle {type=MovieSceneScriptingFloatChannel}
          Scale.X {type=MovieSceneScriptingFloatChannel, keys [[0,1],[0.9847,1.2],[6.0593,1]]}
          Scale.Y {type=MovieSceneScriptingFloatChannel, keys [[0,1],[1,1.2136],[6.0593,1]]}
          Shear.X {type=MovieSceneScriptingFloatChannel}
          Shear.Y {type=MovieSceneScriptingFloatChannel}
         ]
        }]
      }]
    }]}
 }
 Selected_1 {
  binding_count=5
  length_seconds=1
  object=&self:Selected_1
  sequence {bindings [
    CommonBorder_0 {
     id=#1
     possessed_class=CommonBorder
     tracks [
      {
       type=MovieSceneFloatVectorTrack
       display_name=Desired Size Scale
       sections [{
         type=MovieSceneFloatVectorSection
         channels [
          Vector.X {type=MovieSceneScriptingFloatChannel}
          Vector.Y {type=MovieSceneScriptingFloatChannel, keys [[0,0],[20,1]]}
         ]
        }]
      }
      {
       type=MovieSceneFloatTrack
       display_name=Render Opacity
       sections [{type=MovieSceneFloatSection, channels [None {type=MovieSceneScriptingFloatChannel, keys [[0,0],[1.994,1]]}]}]
      }
     ]
    }
    SizeBox_Root {
     id=#2
     possessed_class=SizeBox
     tracks [
      {
       type=MovieSceneFloatTrack
       display_name=Width Override
       sections [{type=MovieSceneFloatSection, channels [None {type=MovieSceneScriptingFloatChannel, keys [[0,384],[20.0003,435]]}]}]
      }
      {
       type=MovieScene2DTransformTrack
       sections [{
         type=MovieScene2DTransformSection
         end_frame=20
         start_frame=0
         channels [
          Translation.X {type=MovieSceneScriptingFloatChannel}
          Translation.Y {type=MovieSceneScriptingFloatChannel}
          Angle {type=MovieSceneScriptingFloatChannel}
          Scale.X {type=MovieSceneScriptingFloatChannel, keys [[0,0.8],[20,1]]}
          Scale.Y {type=MovieSceneScriptingFloatChannel, keys [[0,0.8],[20,1]]}
          Shear.X {type=MovieSceneScriptingFloatChannel}
          Shear.Y {type=MovieSceneScriptingFloatChannel}
         ]
        }]
      }
     ]
    }
    BorderSlot_0 {
     id=#3
     parent_id=#2
     possessed_class=BorderSlot
     tracks [{
       type=MovieSceneMarginTrack
       display_name=Padding
       sections [{
         type=MovieSceneMarginSection
         end_frame=20
         start_frame=0
         channels [
          Left {type=MovieSceneScriptingFloatChannel, keys [[0,-26.0614],[20,12]]}
          Top {type=MovieSceneScriptingFloatChannel, const [0,2]}
          Right {type=MovieSceneScriptingFloatChannel, keys [[0,-26.0438],[20,12]]}
          Bottom {type=MovieSceneScriptingFloatChannel, const [0,2]}
         ]
        }]
      }]
    }
    Image_ColorFill {
     id=#4
     possessed_class=Image
     tracks [{
       type=MovieSceneFloatTrack
       display_name=Render Opacity
       sections [{type=MovieSceneFloatSection, end_frame=20, start_frame=0, channels [None {type=MovieSceneScriptingFloatChannel, keys [[0,0.45],[20,0]]}]}]
      }]
    }
    Image_PanelBorder {
     id=#5
     possessed_class=Image
     tracks [
      {
       type=MovieSceneWidgetMaterialTrack
       sections [{
         type=MovieSceneParameterSection
         channels [
          BottomEdge_1 {type=MovieSceneScriptingFloatChannel, keys [[0,0.05],[20,0.0825]]}
          LeftEdge_1 {type=MovieSceneScriptingFloatChannel, keys [[0,0.0175],[20,0.024]]}
          RightEdge_1 {type=MovieSceneScriptingFloatChannel, keys [[0,0.02],[20,0.027]]}
          TopEdge_1 {type=MovieSceneScriptingFloatChannel, keys [[0,0.075],[20,0.1]]}
         ]
        }]
      }
      {
       type=MovieSceneFloatTrack
       display_name=Render Opacity
       sections [{type=MovieSceneFloatSection, end_frame=20, start_frame=0, channels [None {type=MovieSceneScriptingFloatChannel}]}]
      }
      {
       type=MovieSceneColorTrack
       display_name=Color and Opacity
       sections [{
         type=MovieSceneColorSection
         end_frame=20
         start_frame=0
         channels [
          Color.R {type=MovieSceneScriptingFloatChannel, keys [[0,0],[20,1]]}
          Color.G {type=MovieSceneScriptingFloatChannel, keys [[0,0.4],[20,1]]}
          Color.B {type=MovieSceneScriptingFloatChannel, const [1,1]}
          Color.A {type=MovieSceneScriptingFloatChannel, const [1,2]}
         ]
        }]
      }
     ]
    }
   ]}
 }
 Hover {
  binding_count=5
  length_seconds=1
  object=&self:Hover
  sequence {bindings [
    CommonBorder_0 {
     id=#1
     possessed_class=CommonBorder
     tracks [
      {
       type=MovieSceneFloatVectorTrack
       display_name=Desired Size Scale
       sections [{
         type=MovieSceneFloatVectorSection
         channels [
          Vector.X {type=MovieSceneScriptingFloatChannel}
          Vector.Y {type=MovieSceneScriptingFloatChannel, const [0,1]}
         ]
        }]
      }
      {
       type=MovieSceneFloatTrack
       display_name=Render Opacity
       sections [{type=MovieSceneFloatSection, channels [None {type=MovieSceneScriptingFloatChannel, const [0,1]}]}]
      }
     ]
    }
    SizeBox_Root {
     id=#2
     possessed_class=SizeBox
     tracks [
      {
       type=MovieSceneFloatTrack
       display_name=Width Override
       sections [{type=MovieSceneFloatSection, channels [None {type=MovieSceneScriptingFloatChannel, const [384,1]}]}]
      }
      {
       type=MovieScene2DTransformTrack
       sections [{
         type=MovieScene2DTransformSection
         end_frame=20
         start_frame=0
         channels [
          Translation.X {type=MovieSceneScriptingFloatChannel}
          Translation.Y {type=MovieSceneScriptingFloatChannel}
          Angle {type=MovieSceneScriptingFloatChannel}
          Scale.X {type=MovieSceneScriptingFloatChannel, const [0.8,1]}
          Scale.Y {type=MovieSceneScriptingFloatChannel, const [0.8,1]}
          Shear.X {type=MovieSceneScriptingFloatChannel}
          Shear.Y {type=MovieSceneScriptingFloatChannel}
         ]
        }]
      }
     ]
    }
    BorderSlot_0 {
     id=#3
     parent_id=#2
     possessed_class=BorderSlot
     tracks [{
       type=MovieSceneMarginTrack
       display_name=Padding
       sections [{
         type=MovieSceneMarginSection
         end_frame=20
         start_frame=0
         channels [
          Left {type=MovieSceneScriptingFloatChannel, const [-26.0614,1]}
          Top {type=MovieSceneScriptingFloatChannel, const [0,1]}
          Right {type=MovieSceneScriptingFloatChannel, const [-26.0438,1]}
          Bottom {type=MovieSceneScriptingFloatChannel, const [0,1]}
         ]
        }]
      }]
    }
    Image_ColorFill {
     id=#4
     possessed_class=Image
     tracks [{
       type=MovieSceneFloatTrack
       display_name=Render Opacity
       sections [{type=MovieSceneFloatSection, end_frame=20, start_frame=0, channels [None {type=MovieSceneScriptingFloatChannel, keys [[0,0.45],[20,0.25]]}]}]
      }]
    }
    Image_PanelBorder {
     id=#5
     possessed_class=Image
     tracks [
      {
       type=MovieSceneWidgetMaterialTrack
       sections [{
         type=MovieSceneParameterSection
         channels [
          BottomEdge_1 {type=MovieSceneScriptingFloatChannel, const [0.05,1]}
          LeftEdge_1 {type=MovieSceneScriptingFloatChannel, const [0.0175,1]}
          RightEdge_1 {type=MovieSceneScriptingFloatChannel, const [0.02,1]}
          TopEdge_1 {type=MovieSceneScriptingFloatChannel, const [0.075,1]}
         ]
        }]
      }
      {
       type=MovieSceneFloatTrack
       display_name=Render Opacity
       sections [{type=MovieSceneFloatSection, end_frame=20, start_frame=0, channels [None {type=MovieSceneScriptingFloatChannel}]}]
      }
      {
       type=MovieSceneColorTrack
       display_name=Color and Opacity
       sections [{
         type=MovieSceneColorSection
         end_frame=20
         start_frame=0
         channels [
          Color.R {type=MovieSceneScriptingFloatChannel, const [0,1]}
          Color.G {type=MovieSceneScriptingFloatChannel, keys [[0,0.4],[20,0.6666]]}
          Color.B {type=MovieSceneScriptingFloatChannel, const [1,1]}
          Color.A {type=MovieSceneScriptingFloatChannel, const [1,2]}
         ]
        }]
      }
     ]
    }
   ]}
 }
]
graphs [EventGraph {nodes [
   "On Button Clicked (CreateServerPrivacy)"!() -> (OutputDelegate, then, Button:Common Button Legacy Object Reference) {is=title K2Node_ComponentBoundEvent}
   n1 = Get NewServerIsPrivate() -> (NewServerIsPrivate:bool)
   > "SetText (Text)"(
    self=Get TextPrivacyToggle() -> (TextPrivacyToggle:Common Text Object Reference)
    InText=Select(Option 0:Text="LOCTEXT(\"INVITE ONLY\")", Option 1:Text="LOCTEXT(\"FRIENDS OF FRIENDS\")", Index=@n1) -> (ReturnValue:Text)
   )
   > PlayAnimation(InAnimation=Get PrivacyToggle() -> (PrivacyToggle:Widget Animation Object Reference))
   > Set NewServerIsPrivate(NewServerIsPrivate:bool=Select(Option 0:bool=true, Option 1:bool=false, Index=@n1) -> (ReturnValue:bool)) -> (Output_Get:bool)
   "On Button Hovered (CreateServerPrivacy)"!() -> (OutputDelegate, then, Button:Common Button Legacy Object Reference) {is=title K2Node_ComponentBoundEvent}
   n2 = Get HoverToggle() -> (HoverToggle:Widget Animation Object Reference)
   > PlayAnimationForward(InAnimation=@n2)
   "On Button Unhovered (CreateServerPrivacy)"!() -> (OutputDelegate, then, Button:Common Button Legacy Object Reference) {is=title K2Node_ComponentBoundEvent}
   > PlayAnimationReverse(InAnimation=@n2)
   n3 = InitializeSelectedButton() {is=title K2Node_CustomEvent}
   n4 = Get Selected_1() -> (Selected_1:Widget Animation Object Reference)
   PlayAnimationForward(execute=[@n11.then, @n3.then], InAnimation=@n4, PlaybackSpeed=6)
   Event Construct()
   n5 = Get Selected_1()
   > PlayAnimation(InAnimation=@n5, StartAtTime=GetEndTime(self=@n5), PlayMode=Reverse, PlaybackSpeed=3)
   n6 = Event OnLaunchButtonEnableChanged!() -> (OutputDelegate, then, bEnableValue:bool) {is=title K2Node_Event}
   > n7 = Cast To PrimarySkewButton!(execute, Object=Get Button_Launch() -> (Button_Launch:Common Button Legacy Object Reference)) -> (then, CastFailed, AsPrimary Skew Button:Primary Skew Button Object Reference) {is=title K2Node_DynamicCast}
   > "SetText (Text)"(
    self=Get CenterButtonTextWidget!(self:Primary Skew Button Object Reference=@n7) -> (CenterButtonTextWidget:Common Text Object Reference)
    InText=Select(Option 0:Text="LOCTEXT(\"DISABLED\")", Option 1:Text="LOCTEXT(\"LAUNCH\")", Index=@n6.bEnableValue) -> (ReturnValue:Text)
   )
   Event On Selected()
   n8 = Get Lauch_scalebox() -> (Lauch_scalebox:Scale Box Object Reference)
   > SetVisibility(self=@n8, InVisibility=SelfHitTestInvisible)
   n9 = Get WarningOverlay() -> (WarningOverlay:Overlay Object Reference)
   > SetVisibility(self=@n9, InVisibility=SelfHitTestInvisible)
   n10 = Get ActionDescription() -> (ActionDescription:Common Border Object Reference)
   > n11 = SetVisibility(self=@n10, InVisibility=SelfHitTestInvisible)
   Event On Deselected()
   > SetVisibility(self=@n8, InVisibility=Collapsed)
   > SetVisibility(self=@n9, InVisibility=Collapsed)
   > SetVisibility(self=@n10, InVisibility=Collapsed)
   > PlayAnimationReverse(InAnimation=@n4, PlaybackSpeed=6)
   > Branch(Condition=Get PlayersShown() -> (PlayersShown:bool))
   > PlayAnimationReverse(InAnimation=Get ViewPlayers() -> (ViewPlayers:Widget Animation Object Reference), PlaybackSpeed=4)
   Event On Hovered()
   n12 = IsUsingTouch(PlayerController=GetOwningPlayer())
   > n13 = Branch(Condition=@n12)
   n14 = GetSelected!(self:Common Button Base Object Reference) -> (ReturnValue:bool)
   n15 = Branch(execute=@n13.else, Condition=@n14)
   n16 = IsAnimationPlaying(InAnimation=Get Selected_1())
   n17 = Branch(execute=@n15.else, Condition=@n16)
   n18 = Get Hover() -> (Hover:Widget Animation Object Reference)
   PlayAnimationForward(execute=@n17.else, InAnimation=@n18, PlaybackSpeed=4)
   Event On Unhovered()
   > n19 = Branch(Condition=@n12)
   n20 = Branch(execute=@n19.else, Condition=@n14)
   n21 = Branch(execute=@n20.else, Condition=@n16)
   PlayAnimationReverse(execute=@n21.else, InAnimation=@n18, PlaybackSpeed=4)
   n22 = Length(TargetArray:Array of Texture Object References=Get TEMPBGArray() -> (TEMPBGArray:Array of Texture Object References))
   Set BGImage!(
    execute
    self
    ArrayIdx:int=RandomIntegerInRange!(Min:int=0, Max:int=int - int(A=@n22, B=1) {is=title K2Node_CallFunction}) -> (ReturnValue:int=0)
   ) -> (then) {is=title K2Node_CallFunction}
   Event PreConstruct()
  ]}]
variables [
 NewServerIsPrivate {type=boolean}
 PlayersShown {type=boolean}
 TEMPBGArray {
  type="array<Texture>"
  default [
   /FortniteCoreUI/UI/Creative/Materials/ServerImageAssets/T_UI_Creative_ServerImage_1
   /FortniteCoreUI/UI/Creative/Materials/ServerImageAssets/T_UI_Creative_ServerImage_2
   /FortniteCoreUI/UI/Creative/Materials/ServerImageAssets/T_UI_Creative_ServerImage_3
   /FortniteCoreUI/UI/Creative/Materials/ServerImageAssets/T_UI_Creative_ServerImage_4
   /FortniteCoreUI/UI/Creative/Materials/ServerImageAssets/T_UI_Creative_ServerImage_5
   /FortniteCoreUI/UI/Creative/Materials/ServerImageAssets/T_UI_Creative_ServerImage_6
   /FortniteCoreUI/UI/Creative/Materials/ServerImageAssets/T_UI_Creative_ServerImage_7
   /FortniteCoreUI/UI/Creative/Materials/ServerImageAssets/T_UI_Creative_ServerImage_8
   /FortniteCoreUI/UI/Creative/Materials/ServerImageAssets/T_UI_Creative_ServerImage_9
   /FortniteCoreUI/UI/Creative/Materials/ServerImageAssets/T_UI_Creative_ServerImage_14
   /FortniteCoreUI/UI/Creative/Materials/ServerImageAssets/T_UI_Creative_ServerImage_20
   /FortniteCoreUI/UI/Creative/Materials/ServerImageAssets/T_UI_Creative_ServerImage_21
   /FortniteCoreUI/UI/Creative/Materials/ServerImageAssets/T_UI_Creative_ServerImage_22
   /FortniteCoreUI/UI/Creative/Materials/ServerImageAssets/T_UI_Creative_ServerImage_28
   /FortniteCoreUI/UI/Creative/Materials/ServerImageAssets/T_UI_Creative_ServerImage_26
   /FortniteCoreUI/UI/Creative/Materials/ServerImageAssets/T_UI_Creative_ServerImage_35
   /FortniteCoreUI/UI/Creative/Materials/ServerImageAssets/T_UI_Creative_ServerImage_25
  ]
 }
]
widgets {
 info {inheritedWidgetCount=6, parentClass=/Script/FortniteUI.FortPlaylistSelectionActionTile, rootWidgetClass=&ScaleBox, widgetCount=38}
 widgets [
  ScaleBox_0 {widgetClassPath=&ScaleBox, widget=ScaleBox_0, properties {stretch=UserSpecified, userSpecifiedScale=1.2}}
  CommonBorder_4 {
   widgetClassPath=&CommonBorder
   parent=ScaleBox_0
   widget=CommonBorder_4
   properties {horizontalAlignment=HAlign_Center, verticalAlignment=VAlign_Center, padding=<ltrb 12 0 12 0>}
   slot {class=ScaleBoxSlot}
  }
  SizeBox_Root {
   widgetClassPath=&SizeBox
   parent=CommonBorder_4
   widget=SizeBox_Root
   properties {bOverride_HeightOverride=true, bOverride_WidthOverride=true, heightOverride=650, widthOverride=364.3404}
   slot {class=BorderSlot, properties {horizontalAlignment=HAlign_Center, verticalAlignment=VAlign_Center, padding=<ltrb 12 0 12 0>}}
  }
  Overlay_0 {widgetClassPath=&Overlay, parent=SizeBox_Root, widget=Overlay_0, properties {clipping=ClipToBounds}, slot {class=SizeBoxSlot}}
  SizeBox_1 {
   widgetClassPath=&SizeBox
   parent=Overlay_0
   widget=SizeBox_1
   properties {bOverride_HeightOverride=true, bOverride_WidthOverride=true, heightOverride=650, widthOverride=435}
   slot {class=OverlaySlot, properties {horizontalAlignment=HAlign_Center, verticalAlignment=VAlign_Center, padding=<ltrb -100 0 -100 0>}}
  }
  Overlay_2 {widgetClassPath=&Overlay, parent=SizeBox_1, widget=Overlay_2, properties {clipping=ClipToBounds}, slot {class=SizeBoxSlot}}
  Image_Background {
   widgetClassPath=&Image
   bInherited=true
   bIsVariable=true
   parent=Overlay_2
   widget=Image_Background
   properties {brush {
     bIsDynamicallyLoaded=false
     drawAs=Image
     imageType=NoImage
     mirroring=NoMirror
     resourceName=None
     resourceObject=/FortniteCoreUI/UI/Creative/Materials/ServerImageAssets/T_UI_Creative_ServerImage_39
     tiling=NoTile
     imageSize=<512 1024>
     margin=<ltrb 0 0 0 0>
     outlineSettings {
      bUseBrushTransparency=false
      roundingType=HalfHeightRadius
      width=0
      color {colorUseRule=UseColor_Specified, specifiedColor=<rgba 0 0 0 0>}
      cornerRadii=<0 0 0 0>
     }
     tintColor {colorUseRule=UseColor_Specified, specifiedColor=<rgba 1 1 1 1>}
     uVRegion {bIsValid=false, max=<0 0>, min=<0 0>}
    }}
   slot {class=OverlaySlot, properties {horizontalAlignment=HAlign_Fill, verticalAlignment=VAlign_Fill}}
  }
  BackgroundBlur_0 {
   widgetClassPath=&BackgroundBlur
   parent=Overlay_2
   widget=BackgroundBlur_0
   slot {class=OverlaySlot, properties {horizontalAlignment=HAlign_Fill, verticalAlignment=VAlign_Fill}}
  }
  Image_ColorFill {
   widgetClassPath=&Image
   bIsVariable=true
   parent=Overlay_2
   widget=Image_ColorFill
   properties {renderOpacity=0, colorAndOpacity=<rgba 0.05 0.2083 1 1>}
   slot {class=OverlaySlot, properties {horizontalAlignment=HAlign_Fill, verticalAlignment=VAlign_Fill}}
  }
  Image_Gradient {
   widgetClassPath=&Image
   bIsVariable=true
   parent=Overlay_2
   widget=Image_Gradient
   properties {
    brush {
     bIsDynamicallyLoaded=false
     drawAs=Image
     imageType=NoImage
     mirroring=NoMirror
     resourceName=None
     resourceObject=/FortniteCoreUI/UI/Creative/Materials/M_UI_CreativeServerGradient
     tiling=NoTile
     imageSize=<32 340.801>
     margin=<ltrb 0 0 0 0>
     outlineSettings {
      bUseBrushTransparency=false
      roundingType=HalfHeightRadius
      width=0
      color {colorUseRule=UseColor_Specified, specifiedColor=<rgba 0 0 0 0>}
      cornerRadii=<0 0 0 0>
     }
     tintColor {colorUseRule=UseColor_Specified, specifiedColor=<rgba 1 1 1 1>}
     uVRegion {bIsValid=false, max=<0 0>, min=<0 0>}
    }
    colorAndOpacity=<rgba 0 0.1413 0.5089 1>
   }
   slot {class=OverlaySlot, properties {horizontalAlignment=HAlign_Fill, verticalAlignment=VAlign_Bottom}}
  }
  OverlaySpikes {
   widgetClassPath=&Overlay
   parent=Overlay_2
   widget=OverlaySpikes
   properties {renderOpacity=0}
   slot {class=OverlaySlot, properties {horizontalAlignment=HAlign_Fill, verticalAlignment=VAlign_Fill}}
  }
  spike {
   widgetClassPath=&Image
   parent=OverlaySpikes
   widget=spike
   properties {brush {
     bIsDynamicallyLoaded=false
     drawAs=Image
     imageType=NoImage
     mirroring=NoMirror
     resourceName=None
     resourceObject=/FortniteCoreUI/UI/Creative/Materials/M_UI_ActionTriangle_Inst
     tiling=NoTile
     imageSize=<48 48>
     margin=<ltrb 0 0 0 0>
     outlineSettings {
      bUseBrushTransparency=false
      roundingType=HalfHeightRadius
      width=0
      color {colorUseRule=UseColor_Specified, specifiedColor=<rgba 0 0 0 0>}
      cornerRadii=<0 0 0 0>
     }
     tintColor {colorUseRule=UseColor_Specified, specifiedColor=<rgba 1 1 1 1>}
     uVRegion {bIsValid=false, max=<0 0>, min=<0 0>}
    }}
   slot {class=OverlaySlot}
  }
  Image_112 {
   widgetClassPath=&Image
   bIsVariable=true
   parent=OverlaySpikes
   widget=Image_112
   properties {
    brush {
     bIsDynamicallyLoaded=false
     drawAs=Image
     imageType=NoImage
     mirroring=NoMirror
     resourceName=None
     resourceObject=&M_UI_ActionTriangle_Inst
     tiling=NoTile
     imageSize=<28 28>
     margin=<ltrb 0 0 0 0>
     outlineSettings {
      bUseBrushTransparency=false
      roundingType=HalfHeightRadius
      width=0
      color {colorUseRule=UseColor_Specified, specifiedColor=<rgba 0 0 0 0>}
      cornerRadii=<0 0 0 0>
     }
     tintColor {colorUseRule=UseColor_Specified, specifiedColor=<rgba 1 1 1 1>}
     uVRegion {bIsValid=false, max=<0 0>, min=<0 0>}
    }
    renderTransform {angle=0, scale=<-1 1>, shear=<0 0>, translation=<0 22>}
   }
   slot {class=OverlaySlot, properties {horizontalAlignment=HAlign_Right}}
  }
  Image_113 {
   widgetClassPath=&Image
   bIsVariable=true
   parent=OverlaySpikes
   widget=Image_113
   properties {
    brush {
     bIsDynamicallyLoaded=false
     drawAs=Image
     imageType=NoImage
     mirroring=NoMirror
     resourceName=None
     resourceObject=&M_UI_ActionTriangle_Inst
     tiling=NoTile
     imageSize=<48.7241 55.9895>
     margin=<ltrb 0 0 0 0>
     outlineSettings {
      bUseBrushTransparency=false
      roundingType=HalfHeightRadius
      width=0
      color {colorUseRule=UseColor_Specified, specifiedColor=<rgba 0 0 0 0>}
      cornerRadii=<0 0 0 0>
     }
     tintColor {colorUseRule=UseColor_Specified, specifiedColor=<rgba 1 1 1 1>}
     uVRegion {bIsValid=false, max=<0 0>, min=<0 0>}
    }
    renderTransform {angle=0, scale=<-1 1>, shear=<0 0>, translation=<0 0>}
   }
   slot {class=OverlaySlot, properties {horizontalAlignment=HAlign_Right}}
  }
  Image_114 {
   widgetClassPath=&Image
   bIsVariable=true
   parent=OverlaySpikes
   widget=Image_114
   properties {
    brush {
     bIsDynamicallyLoaded=false
     drawAs=Image
     imageType=NoImage
     mirroring=NoMirror
     resourceName=None
     resourceObject=&M_UI_ActionTriangle_Inst
     tiling=NoTile
     imageSize=<64 64>
     margin=<ltrb 0 0 0 0>
     outlineSettings {
      bUseBrushTransparency=false
      roundingType=HalfHeightRadius
      width=0
      color {colorUseRule=UseColor_Specified, specifiedColor=<rgba 0 0 0 0>}
      cornerRadii=<0 0 0 0>
     }
     tintColor {colorUseRule=UseColor_Specified, specifiedColor=<rgba 1 1 1 1>}
     uVRegion {bIsValid=false, max=<0 0>, min=<0 0>}
    }
    renderTransform {angle=0, scale=<1 -1>, shear=<0 0>, translation=<0 0>}
    renderTransformPivot=<0 0.5>
   }
   slot {class=OverlaySlot, properties {verticalAlignment=VAlign_Bottom}}
  }
  Image_115 {
   widgetClassPath=&Image
   bIsVariable=true
   parent=OverlaySpikes
   widget=Image_115
   properties {
    brush {
     bIsDynamicallyLoaded=false
     drawAs=Image
     imageType=NoImage
     mirroring=NoMirror
     resourceName=None
     resourceObject=&M_UI_ActionTriangle_Inst
     tiling=NoTile
     imageSize=<31.3787 32>
     margin=<ltrb 0 0 0 0>
     outlineSettings {
      bUseBrushTransparency=false
      roundingType=HalfHeightRadius
      width=0
      color {colorUseRule=UseColor_Specified, specifiedColor=<rgba 0 0 0 0>}
      cornerRadii=<0 0 0 0>
     }
     tintColor {colorUseRule=UseColor_Specified, specifiedColor=<rgba 1 1 1 1>}
     uVRegion {bIsValid=false, max=<0 0>, min=<0 0>}
    }
    renderTransform {angle=0, scale=<1 -1>, shear=<0 0>, translation=<21 0>}
    renderTransformPivot=<0 0.5>
   }
   slot {class=OverlaySlot, properties {verticalAlignment=VAlign_Bottom}}
  }
  Image_116 {
   widgetClassPath=&Image
   bIsVariable=true
   parent=OverlaySpikes
   widget=Image_116
   properties {
    brush {
     bIsDynamicallyLoaded=false
     drawAs=Image
     imageType=NoImage
     mirroring=NoMirror
     resourceName=None
     resourceObject=&M_UI_ActionTriangle_Inst
     tiling=NoTile
     imageSize=<53.8867 53.4073>
     margin=<ltrb 0 0 0 0>
     outlineSettings {
      bUseBrushTransparency=false
      roundingType=HalfHeightRadius
      width=0
      color {colorUseRule=UseColor_Specified, specifiedColor=<rgba 0 0 0 0>}
      cornerRadii=<0 0 0 0>
     }
     tintColor {colorUseRule=UseColor_Specified, specifiedColor=<rgba 1 1 1 1>}
     uVRegion {bIsValid=false, max=<0 0>, min=<0 0>}
    }
    renderTransform {angle=0, scale=<-1 -1>, shear=<0 0>, translation=<0 0>}
   }
   slot {class=OverlaySlot, properties {horizontalAlignment=HAlign_Right, verticalAlignment=VAlign_Bottom}}
  }
  Image_PanelBorder {
   widgetClassPath=&Image
   parent=Overlay_0
   widget=Image_PanelBorder
   properties {brush {
     bIsDynamicallyLoaded=false
     drawAs=Box
     imageType=NoImage
     mirroring=NoMirror
     resourceName=None
     resourceObject=/Game/UI/Creative/Materials/M_UI_ServerBorder_Selected
     tiling=NoTile
     imageSize=<32 128>
     margin=<ltrb 0 0.17 0 0.17>
     outlineSettings {
      bUseBrushTransparency=false
      roundingType=HalfHeightRadius
      width=0
      color {colorUseRule=UseColor_Specified, specifiedColor=<rgba 0 0 0 0>}
      cornerRadii=<0 0 0 0>
     }
     tintColor {colorUseRule=UseColor_Specified, specifiedColor=<rgba 1 1 1 1>}
     uVRegion {bIsValid=false, max=<0 0>, min=<0 0>}
    }}
   slot {class=OverlaySlot, properties {horizontalAlignment=HAlign_Fill, verticalAlignment=VAlign_Fill}}
  }
  CommonBorder_3 {
   widgetClassPath=&CommonBorder
   parent=Overlay_0
   widget=CommonBorder_3
   properties {verticalAlignment=VAlign_Bottom, padding=<ltrb 0 0 0 0>}
   slot {class=OverlaySlot, properties {horizontalAlignment=HAlign_Fill, verticalAlignment=VAlign_Fill}}
  }
  CreateServerContent {
   widgetClassPath=&VerticalBox
   bIsVariable=true
   parent=CommonBorder_3
   widget=CreateServerContent
   slot {class=BorderSlot, properties {verticalAlignment=VAlign_Bottom, padding=<ltrb 0 0 0 0>}}
  }
  VerticalBox_1 {
   widgetClassPath=&VerticalBox
   parent=CreateServerContent
   widget=VerticalBox_1
   slot {class=VerticalBoxSlot, properties {verticalAlignment=VAlign_Bottom, padding=<ltrb 0 0 0 20>}}
  }
  CommonBorder_NewServerContent {
   widgetClassPath=&CommonBorder
   bIsVariable=true
   parent=VerticalBox_1
   widget=CommonBorder_NewServerContent
   properties {padding=<ltrb 0 0 0 0>}
   slot {class=VerticalBoxSlot, properties {padding=<ltrb 16 0 16 0>}}
  }
  Text_ActionTitle {
   widgetClassPath=&CommonTextBlock
   bInherited=true
   parent=CommonBorder_NewServerContent
   widget=Text_ActionTitle
   properties {
    autoWrapText=true
    justification=Center
    lineHeightPercentage=0.8
    text=CREATE A BATTLELAB
    font {
     bForceMonospaced=false
     bMaterialIsStencil=false
     fontMaterial=null
     fontObject=&Burbank
     letterSpacing=0
     monospacedWidth=1
     size=40
     skewAmount=0
     typefaceFontName=BurbankRegular-Black
     outlineSettings {
      bApplyOutlineToDropShadows=false
      bMiteredCorners=false
      bSeparateFillAlpha=false
      outlineMaterial=null
      outlineSize=2
      outlineColor=<rgba 0.0052 0.0052 0.0056 0.2>
     }
    }
    shadowOffset=<0 0>
   }
   slot {class=BorderSlot, properties {padding=<ltrb 0 0 0 0>}}
  }
  ActionDescription {
   widgetClassPath=&CommonBorder
   bIsVariable=true
   parent=VerticalBox_1
   widget=ActionDescription
   properties {visibility=Collapsed, padding=<ltrb 0 0 0 0>}
   slot {class=VerticalBoxSlot, properties {padding=<ltrb 30 10 30 0>}}
  }
  Text_ActionDescription {
   widgetClassPath=&CommonTextBlock
   bInherited=true
   bIsVariable=true
   parent=ActionDescription
   widget=Text_ActionDescription
   properties {
    justification=Center
    lineHeightPercentage=0.9
    text="Make your ideal game on the Battle Royale Island!"
    wrapTextAt=320
    colorAndOpacity {colorUseRule=UseColor_Specified, specifiedColor=<rgba 0.6038 0.8469 1 1>}
    font {
     bForceMonospaced=false
     bMaterialIsStencil=false
     fontMaterial=null
     fontObject=&BurbankSmall
     letterSpacing=0
     monospacedWidth=1
     size=16
     skewAmount=0
     typefaceFontName=Bold
     outlineSettings {
      bApplyOutlineToDropShadows=false
      bMiteredCorners=false
      bSeparateFillAlpha=false
      outlineMaterial=null
      outlineSize=0
      outlineColor=<rgba 0 0 0 1>
     }
    }
    shadowOffset=<0 0>
   }
   slot {class=BorderSlot, properties {padding=<ltrb 0 0 0 0>}}
  }
  CommonBorder_0 {
   widgetClassPath=&CommonBorder
   parent=VerticalBox_1
   widget=CommonBorder_0
   properties {visibility=Collapsed, padding=<ltrb 0 0 0 0>}
   slot {class=VerticalBoxSlot}
  }
  SizeBox_5 {
   widgetClassPath=&SizeBox
   parent=CommonBorder_0
   widget=SizeBox_5
   properties {bOverride_MinDesiredHeight=true, minDesiredHeight=55}
   slot {class=BorderSlot, properties {padding=<ltrb 0 0 0 0>}}
  }
  Overlay_3 {widgetClassPath=&Overlay, parent=SizeBox_5, widget=Overlay_3, slot {class=SizeBoxSlot}}
  Image_DetailsButtonBG {
   widgetClassPath=&Image
   parent=Overlay_3
   widget=Image_DetailsButtonBG
   properties {
    visibility=SelfHitTestInvisible
    brush {
     bIsDynamicallyLoaded=false
     drawAs=Image
     imageType=NoImage
     mirroring=NoMirror
     resourceName=None
     resourceObject=/Game/UI/Frontend/Showdown/Materials/M_UI_SkewBox_TournamentPanelDetailsBG
     tiling=NoTile
     imageSize=<32 32>
     margin=<ltrb 0 0 0 0>
     outlineSettings {
      bUseBrushTransparency=false
      roundingType=HalfHeightRadius
      width=0
      color {colorUseRule=UseColor_Specified, specifiedColor=<rgba 0 0 0 0>}
      cornerRadii=<0 0 0 0>
     }
     tintColor {colorUseRule=UseColor_Specified, specifiedColor=<rgba 1 1 1 1>}
     uVRegion {bIsValid=false, max=<0 0>, min=<0 0>}
    }
    renderTransformPivot=<0 1>
   }
   slot {class=OverlaySlot, properties {horizontalAlignment=HAlign_Fill, verticalAlignment=VAlign_Fill, padding=<ltrb 0 -5 0 0>}}
  }
  CreateServerPrivacy {
   widgetClassPath=&IconTextButton
   bIsVariable=true
   parent=Overlay_3
   widget=CreateServerPrivacy
   properties {
    buttonText=""
    onButtonBaseClicked=(&self:WidgetTree.CreateServerPrivacy.HandleOnButtonClicked)
    onButtonBaseDoubleClicked=(&self:WidgetTree.CreateServerPrivacy.HandleOnButtonDoubleClicked)
    onButtonBaseHovered=(&self:WidgetTree.CreateServerPrivacy.HandleOnButtonHovered)
    onButtonBaseUnhovered=(&self:WidgetTree.CreateServerPrivacy.HandleOnButtonUnhovered)
    onSelectedChangedBase=(&self:WidgetTree.CreateServerPrivacy.HandleOnSelectedChanged)
    style=/FortUILibrary/Foundation/Buttons/ButtonStyles/ButtonStyle-TextOnlyBase_Empty.ButtonStyle-TextOnlyBase_Empty_C
    visibility=Collapsed
    triggeringInputAction {dataTable=/FortUILibrary/Input/CreativeInputActionDataTable, rowName=ServerToggleOptions}
   }
   slot {class=OverlaySlot, properties {horizontalAlignment=HAlign_Fill, verticalAlignment=VAlign_Fill, padding=<ltrb 15 0 15 0>}}
  }
  ScaleBox_DetailsText {
   widgetClassPath=&ScaleBox
   parent=Overlay_3
   widget=ScaleBox_DetailsText
   properties {stretchDirection=DownOnly}
   slot {class=OverlaySlot, properties {horizontalAlignment=HAlign_Fill, verticalAlignment=VAlign_Fill, padding=<ltrb 30 0 30 0>}}
  }
  TextPrivacyToggle {
   widgetClassPath=&CommonTextBlock
   bIsVariable=true
   parent=ScaleBox_DetailsText
   widget=TextPrivacyToggle
   properties {
    justification=Right
    text=FRIENDS OF FRIENDS
    colorAndOpacity {colorUseRule=UseColor_Specified, specifiedColor=<rgba 0.0116 0.016 0.0395 1>}
    font {
     bForceMonospaced=false
     bMaterialIsStencil=false
     fontMaterial=null
     fontObject=&BurbankSmall
     letterSpacing=0
     monospacedWidth=1
     size=20
     skewAmount=0
     typefaceFontName=Black
     outlineSettings {
      bApplyOutlineToDropShadows=false
      bMiteredCorners=false
      bSeparateFillAlpha=false
      outlineMaterial=null
      outlineSize=0
      outlineColor=<rgba 0 0 0 1>
     }
    }
   }
   slot {class=ScaleBoxSlot}
  }
  SizeBox_2 {
   widgetClassPath=&SizeBox
   parent=VerticalBox_1
   widget=SizeBox_2
   properties {bOverride_HeightOverride=true, heightOverride=64}
   slot {class=VerticalBoxSlot, properties {padding=<ltrb 16 10 16 0>, size {sizeRule=Fill, value=1}}}
  }
  Lauch_scalebox {
   widgetClassPath=&ScaleBox
   bIsVariable=true
   parent=SizeBox_2
   widget=Lauch_scalebox
   properties {visibility=Collapsed}
   slot {class=SizeBoxSlot}
  }
  Button_Launch {
   widgetClassPath=/FortniteCoreUI/Frontend/Showdown/PrimarySkewButton
   bInherited=true
   bIsVariable=true
   parent=Lauch_scalebox
   widget=Button_Launch
   properties {
    buttonNameText=LAUNCH
    onButtonBaseClicked=(&self:WidgetTree.Button_Launch.HandleOnButtonClicked)
    onButtonBaseDoubleClicked=(&self:WidgetTree.Button_Launch.HandleOnButtonDoubleClicked)
    onButtonBaseHovered=(&self:WidgetTree.Button_Launch.HandleOnButtonHovered)
    onButtonBaseUnhovered=(&self:WidgetTree.Button_Launch.HandleOnButtonUnhovered)
    onSelectedChangedBase=(&self:WidgetTree.Button_Launch.HandleOnSelectedChanged)
    triggeringInputAction {dataTable=&InputActionDataTable, rowName=Confirm}
   }
   slot {class=ScaleBoxSlot}
  }
  WarningOverlay {
   widgetClassPath=&Overlay
   bIsVariable=true
   parent=VerticalBox_1
   widget=WarningOverlay
   slot {class=VerticalBoxSlot, properties {padding=<ltrb 0 5 0 0>}}
  }
  Text_ServerWarning {
   widgetClassPath=&CommonTextBlock
   bInherited=true
   bIsVariable=true
   parent=WarningOverlay
   widget=Text_ServerWarning
   properties {
    autoWrapText=true
    justification=Center
    text=WARNING TEXT
    visibility=Collapsed
    colorAndOpacity {colorUseRule=UseColor_Specified, specifiedColor=<rgba 1 0.9301 0 1>}
    font {
     bForceMonospaced=false
     bMaterialIsStencil=false
     fontMaterial=null
     fontObject=&BurbankSmall
     letterSpacing=0
     monospacedWidth=1
     size=12
     skewAmount=0
     typefaceFontName=Black
     outlineSettings {
      bApplyOutlineToDropShadows=false
      bMiteredCorners=false
      bSeparateFillAlpha=false
      outlineMaterial=null
      outlineSize=0
      outlineColor=<rgba 0 0 0 1>
     }
    }
    shadowOffset=<0 0>
   }
   slot {class=OverlaySlot, properties {horizontalAlignment=HAlign_Fill, verticalAlignment=VAlign_Fill}}
  }
  InputActionWidget {widgetClassPath=&CommonActionWidget, bInherited=true, widget=null}
 ]
}
