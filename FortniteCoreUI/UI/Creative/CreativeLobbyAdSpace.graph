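# Creative lobby ad space widget, based on FortCreativeAdsView, that shows up to two CreativeLobbyAd tiles side by side.
# After a short delay on Construct, it counts CreativeAdList and switches on that count to set visibility and populate the tiles.
# One ad collapses Ad2; zero or more than two leave the container collapsed, as the switch's default is unwired.
# Two ads make the container HitTestInvisible, unlike SelfHitTestInvisible for one, blocking tile hit-testing.
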WidgetBlueprint /FortniteCoreUI/UI/Creative/CreativeLobbyAdSpace
graphs [EventGraph {nodes [
   Event Construct()
   > Delay(Duration=0.2)
   > n1 = Set NumCreativeAds(NumCreativeAds:int=Length(TargetArray:Array of Creative Ad Data Structures=Get CreativeAdList() -> (CreativeAdList:Array of Creative Ad Data Structures))) -> (Output_Get:int=0)
   > n2 = Switch on Int!(execute:Exec, Selection:int=@n1) -> (Default:Exec, 0:Exec, 1:Exec, 2:Exec)
   SetVisibility(execute=@n2.0, self=Get CreativeAdContent() -> (CreativeAdContent:Size Box Object Reference), InVisibility=Collapsed)
   SetVisibility(execute=@n2.1, self=Get CreativeAdContent(), InVisibility=SelfHitTestInvisible)
   > PopulateAdTile!(
    execute
    self:Fort Creative Ad Tile Object Reference=Get Ad1() -> (Ad1:Creative Lobby Ad Object Reference)
    Entry:"Creative Ad Data Structure (by ref)"="Get (a ref)"(Array:Array of Creative Ad Data Structures=Get CreativeAdList(), Dimension 1=0) -> (Output:"Creative Ad Data Structure (by ref)")
   ) -> (then)
   > SetVisibility(self=Get Ad2() -> (Ad2:Creative Lobby Ad Object Reference), InVisibility=Collapsed)
   SetVisibility(execute=@n2.2, self=Get CreativeAdContent(), InVisibility=HitTestInvisible)
   > PopulateAdTile(self=Get Ad1(), Entry="Get (a ref)"(Array=Get CreativeAdList(), Dimension 1=0))
   > PopulateAdTile(self=Get Ad2(), Entry="Get (a ref)"(Array=Get CreativeAdList(), Dimension 1=1))
  ]}]
variables [NumCreativeAds {type=integer}]
widgets {
 info {inheritedWidgetCount=0, parentClass=/Script/FortniteUI.FortCreativeAdsView, rootWidgetClass=&SizeBox, widgetCount=4}
 widgets [
  CreativeAdContent {widgetClassPath=&SizeBox, bIsVariable=true, widget=CreativeAdContent, properties {visibility=Collapsed, widthOverride=928}}
  HorizontalBox_0 {widgetClassPath=&HorizontalBox, parent=CreativeAdContent, widget=HorizontalBox_0, slot {class=SizeBoxSlot}}
  Ad1 {
   widgetClassPath=&/CreativeLobbyAd
   bIsVariable=true
   parent=HorizontalBox_0
   widget=Ad1
   slot {class=HorizontalBoxSlot}
  }
  Ad2 {
   widgetClassPath=&CreativeLobbyAd
   bIsVariable=true
   parent=HorizontalBox_0
   widget=Ad2
   slot {class=HorizontalBoxSlot, properties {padding=<ltrb 24 0 0 0>}}
  }
 ]
}
