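# Translucent UI master material, under frontend art, for a two-color radial gradient with parameters for colors, center, radius, density and alpha.
# Normalized UI UVs are remapped to a -1 to 1 range, so zero center is mid-widget, and feed RadialGradientExponential.
# The gradient drives both an additive blend and a lerp of RGB_A and RGB_B.
# The graph shows no links from these blends or Alpha to material outputs.
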Material /Game/Athena/UI/Frontend/Art/M_UI_RadialGradientMaster_Translucent
graph {
 properties {
  bUsedWithStaticMesh=false
  blendMode=BLEND_Translucent
  materialDomain=MD_UI
  pixelDepthOffsetMode=PDOM_Legacy
  usedShadingModels=Default Lit
 }
 nodes [
  n1 = VectorParameter() {parameterName=RGB_A, defaultValue=<rgba 0.0194 0.0666 0.4179 1>}
  n2 = MaterialFunctionCall() {materialFunction=&GetUserInterfaceUV}
  n3 = MaterialFunctionCall(
   UVs=Multiply(A=Subtract(A=@n2.Normalized UV) {constB=0.5}) {constB=2}
   CenterPosition=MaterialFunctionCall(X=ScalarParameter() {parameterName=Center_X}, Y=ScalarParameter() {parameterName=Center_Y}) {materialFunction=&MakeFloat2}.Result
   Radius=ScalarParameter() {defaultValue=1.5, parameterName=Radius}
   Density=ScalarParameter() {defaultValue=2, parameterName=Density}
  ) {materialFunction=&RadialGradientExponential}
  n4 = VectorParameter() {parameterName=RGB_B, defaultValue=<rgba 0 0.4342 0.4233 1>}
  Add(A=@n1.RGB, B=Multiply(A=@n3.RadialGradientExponential, B=@n4.RGB))
  LinearInterpolate(A=@n1.RGB, B=@n4.RGB, Alpha=@n3.RadialGradientExponential)
  ScalarParameter() {defaultValue=1, parameterName=Alpha}
 ]
}
