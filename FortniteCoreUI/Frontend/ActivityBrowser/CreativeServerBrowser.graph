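# Creative server browser screen where the player selects a Creative server, a layout over native FortServerBrowserCreative with no logic in its event graph.
# An inherited list view shows server info, server action and playlist action tiles over a blurred gradient background, with a Cancel button below.
# The header holds the title, an ad space and a touch-only close button.
# Add-friends and empty island-code panels start collapsed.
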WidgetBlueprint /FortniteCoreUI/Frontend/ActivityBrowser/CreativeServerBrowser
graphs [EventGraph {nodes [
   Event PreConstruct()
   Event Tick()
  ]}]
variables []
widgets {
 info {inheritedWidgetCount=3, parentClass=/Script/FortniteUI.FortServerBrowserCreative, rootWidgetClass=&BackgroundBlur, widgetCount=35}
 widgets [
  BackgroundBlur_0 {
   widgetClassPath=&BackgroundBlur
   widget=BackgroundBlur_0
   properties {
    blurStrength=25
    visibility=Visible
    lowQualityFallbackBrush {
     bIsDynamicallyLoaded=false
     drawAs=Image
     imageType=NoImage
     mirroring=NoMirror
     resourceName=None
     resourceObject=null
     tiling=NoTile
     imageSize=<0 0>
     margin=<ltrb 0 0 0 0>
     outlineSettings {
      bUseBrushTransparency=false
      roundingType=HalfHeightRadius
      width=0
      color {colorUseRule=UseColor_Specified, specifiedColor=<rgba 0 0 0 0>}
      cornerRadii=<0 0 0 0>
     }
     tintColor {colorUseRule=UseColor_Specified, specifiedColor=<rgba 0 0.15 0.9 0.85>}
     uVRegion {bIsValid=false, max=<0 0>, min=<0 0>}
    }
   }
  }
  Overlay_2 {
   widgetClassPath=&Overlay
   parent=BackgroundBlur_0
   widget=Overlay_2
   slot {class=BackgroundBlurSlot, properties {padding=<ltrb 0 0 0 0>}}
  }
  Overlay_3 {
   widgetClassPath=&Overlay
   parent=Overlay_2
   widget=Overlay_3
   slot {class=OverlaySlot, properties {horizontalAlignment=HAlign_Fill, verticalAlignment=VAlign_Fill}}
  }
  BackgroundGradient_1 {
   widgetClassPath=&Image
   parent=Overlay_3
   widget=BackgroundGradient_1
   properties {
    visibility=SelfHitTestInvisible
    brush {
     bIsDynamicallyLoaded=false
     drawAs=Image
     imageType=NoImage
     mirroring=NoMirror
     resourceName=None
     resourceObject=/Game/Athena/UI/Frontend/Art/M_UI_RadialGradientMaster_Translucent
     tiling=NoTile
     imageSize=<2160 2160>
     margin=<ltrb 0 0 0 0>
     outlineSettings {
      bUseBrushTransparency=false
      roundingType=HalfHeightRadius
      width=0
      color {colorUseRule=UseColor_Specified, specifiedColor=<rgba 0 0 0 0>}
      cornerRadii=<0 0 0 0>
     }
     tintColor {colorUseRule=UseColor_Specified, specifiedColor=<rgba 1 1 1 1>}
     uVRegion {bIsValid=false, max=<0 0>, min=<0 0>}
    }
    renderTransform {angle=0, scale=<1.25 2.333>, shear=<0 0>, translation=<0 512>}
   }
   slot {class=OverlaySlot}
  }
  Image_IdlePulse {
   widgetClassPath=&Image
   bIsVariable=true
   parent=Overlay_3
   widget=Image_IdlePulse
   properties {
    visibility=SelfHitTestInvisible
    brush {
     bIsDynamicallyLoaded=false
     drawAs=NoDrawType
     imageType=NoImage
     mirroring=NoMirror
     resourceName=None
     resourceObject=/FortniteCoreUI/UI/Login/SubGameSelect/Materials/M_UI_BGRing_Inst
     tiling=NoTile
     imageSize=<32 32>
     margin=<ltrb 0 0 0 0>
     outlineSettings {
      bUseBrushTransparency=false
      roundingType=HalfHeightRadius
      width=0
      color {colorUseRule=UseColor_Specified, specifiedColor=<rgba 0 0 0 0>}
      cornerRadii=<0 0 0 0>
     }
     tintColor {colorUseRule=UseColor_Specified, specifiedColor=<rgba 1 1 1 1>}
     uVRegion {bIsValid=false, max=<0 0>, min=<0 0>}
    }
    colorAndOpacity=<rgba 0 1 0.5333 0.6571>
    renderTransform {angle=0, scale=<1 1.666>, shear=<0 0>, translation=<0 0>}
   }
   slot {class=OverlaySlot, properties {horizontalAlignment=HAlign_Fill, verticalAlignment=VAlign_Fill}}
  }
  SafeZone_0 {
   widgetClassPath=&SafeZone
   bIsVariable=true
   parent=Overlay_2
   widget=SafeZone_0
   slot {class=OverlaySlot, properties {horizontalAlignment=HAlign_Fill, verticalAlignment=VAlign_Fill}}
  }
  ScaleBox_0 {widgetClassPath=&ScaleBox, parent=SafeZone_0, widget=ScaleBox_0, properties {stretch=ScaleBySafeZone}, slot {class=SafeZoneSlot}}
  Overlay {
   widgetClassPath=&Overlay
   parent=ScaleBox_0
   widget=Overlay
   slot {class=ScaleBoxSlot, properties {horizontalAlignment=HAlign_Fill, verticalAlignment=VAlign_Fill}}
  }
  ListView_ServerBrowser {
   widgetClassPath=/Script/FortniteUI.FortServerBrowserListView
   bInherited=true
   bIsVariable=true
   parent=Overlay
   widget=ListView_ServerBrowser
   properties {
    bEnableScrollAnimation=true
    clipping=Inherit
    creativeServerInfoTileEntryClass=&/ServerBrowserInfoTile_Creative.ServerBrowserInfoTile_Creative_C
    entryWidgetClass=&ServerBrowserInfoTile_Creative_C
    navigation=&self:WidgetTree.ListView_ServerBrowser.WidgetNavigation_0
    numDesignerPreviewEntries=4
    playlistSelectionActionTileEntryClass=&/ServerBrowserPlaylistActionTile.ServerBrowserPlaylistActionTile_C
    serverActionTileEntryClass=&/ServerBrowserActionTile.ServerBrowserActionTile_C
    wheelScrollMultiplier=10
   }
   slot {class=OverlaySlot, properties {horizontalAlignment=HAlign_Center, verticalAlignment=VAlign_Center, padding=<ltrb 0 50 0 0>}}
  }
  SizeBox_5 {
   widgetClassPath=&SizeBox
   parent=Overlay
   widget=SizeBox_5
   properties {bOverride_WidthOverride=true, widthOverride=506.2222}
   slot {class=OverlaySlot, properties {horizontalAlignment=HAlign_Center, verticalAlignment=VAlign_Bottom, padding=<ltrb 0 16 0 30>}}
  }
  VerticalBox_1 {widgetClassPath=&VerticalBox, parent=SizeBox_5, widget=VerticalBox_1, slot {class=SizeBoxSlot}}
  AddFriendsContent {
   widgetClassPath=&SizeBox
   bIsVariable=true
   parent=Overlay
   widget=AddFriendsContent
   properties {bOverride_WidthOverride=true, visibility=Collapsed, widthOverride=500}
   slot {class=OverlaySlot, properties {verticalAlignment=VAlign_Bottom, padding=<ltrb 25 25 25 25>}}
  }
  Overlay_1 {widgetClassPath=&Overlay, parent=AddFriendsContent, widget=Overlay_1, slot {class=SizeBoxSlot}}
  VerticalBox {
   widgetClassPath=&VerticalBox
   parent=Overlay_1
   widget=VerticalBox
   properties {renderTransformPivot=<0.5 1>}
   slot {class=OverlaySlot, properties {horizontalAlignment=HAlign_Fill}}
  }
  AddFriendsText {
   widgetClassPath=&CommonTextBlock
   bIsVariable=true
   parent=VerticalBox
   widget=AddFriendsText
   properties {
    autoWrapText=true
    style=/FortUILibrary/Styles/Typography/Foundation/TextStyle-Header-S.TextStyle-Header-S_C
    text="Add your {Platform} friends as Epic friends to see their Creative servers, too!"
    font {
     bForceMonospaced=false
     bMaterialIsStencil=false
     fontMaterial=null
     fontObject=&Burbank
     letterSpacing=0
     monospacedWidth=1
     size=24
     skewAmount=0
     typefaceFontName=Burbank-Black
     outlineSettings {
      bApplyOutlineToDropShadows=false
      bMiteredCorners=false
      bSeparateFillAlpha=false
      outlineMaterial=null
      outlineSize=0
      outlineColor=<rgba 0 0 0 1>
     }
    }
    shadowOffset=<0 0>
   }
   slot {class=VerticalBoxSlot, properties {padding=<ltrb 0 10 0 0>}}
  }
  AddFriendsButton {
   widgetClassPath=&IconTextButton
   bIsVariable=true
   parent=VerticalBox
   widget=AddFriendsButton
   properties {
    bIsEnabled=false
    buttonText=ADD FRIENDS
    contentAlignment=HAlign_Center
    onButtonBaseClicked=(&self:WidgetTree.AddFriendsButton.HandleOnButtonClicked)
    onButtonBaseDoubleClicked=(&self:WidgetTree.AddFriendsButton.HandleOnButtonDoubleClicked)
    onButtonBaseHovered=(&self:WidgetTree.AddFriendsButton.HandleOnButtonHovered)
    onButtonBaseUnhovered=(&self:WidgetTree.AddFriendsButton.HandleOnButtonUnhovered)
    onSelectedChangedBase=(&self:WidgetTree.AddFriendsButton.HandleOnSelectedChanged)
    style=/FortUILibrary/Foundation/Buttons/ButtonStyles/Skew/ButtonStyle-Skew2.ButtonStyle-Skew2_C
    triggeringInputAction {dataTable=/FortUILibrary/Input/CreativeInputActionDataTable, rowName=ServerSelectAddFriends}
   }
   slot {class=VerticalBoxSlot}
  }
  CommonTextBlock_0 {
   widgetClassPath=&CommonTextBlock
   parent=Overlay_1
   widget=CommonTextBlock_0
   properties {
    autoWrapText=true
    justification=Center
    renderOpacity=0
    style=/FortUILibrary/Styles/Typography/Foundation/TextStyle-Header-L.TextStyle-Header-L_C
    text="ADDING FRIENDS!"
    font {
     bForceMonospaced=false
     bMaterialIsStencil=false
     fontMaterial=null
     fontObject=&Burbank
     letterSpacing=0
     monospacedWidth=1
     size=48
     skewAmount=0
     typefaceFontName=Burbank-Black
     outlineSettings {
      bApplyOutlineToDropShadows=false
      bMiteredCorners=false
      bSeparateFillAlpha=false
      outlineMaterial=null
      outlineSize=0
      outlineColor=<rgba 0 0 0 1>
     }
    }
    shadowOffset=<0 0>
   }
   slot {class=OverlaySlot, properties {horizontalAlignment=HAlign_Fill, verticalAlignment=VAlign_Bottom}}
  }
  IslandCodeContent {
   widgetClassPath=&SizeBox
   bIsVariable=true
   parent=Overlay
   widget=IslandCodeContent
   properties {bOverride_WidthOverride=true, maxDesiredWidth=500, visibility=Collapsed, widthOverride=600}
   slot {class=OverlaySlot, properties {verticalAlignment=VAlign_Bottom, padding=<ltrb 64 0 0 64>}}
  }
  HorizontalBox_0 {
   widgetClassPath=&HorizontalBox
   parent=Overlay
   widget=HorizontalBox_0
   slot {class=OverlaySlot, properties {horizontalAlignment=HAlign_Right, verticalAlignment=VAlign_Bottom, padding=<ltrb 0 0 35 20>}}
  }
  Button_Cancel {
   widgetClassPath=/FortUILibrary/Foundation/Buttons/SoloButton
   bInherited=true
   bIsVariable=true
   parent=HorizontalBox_0
   widget=Button_Cancel
   properties {
    button Description=CANCEL
    onButtonBaseClicked=(&self:WidgetTree.Button_Cancel.HandleOnButtonClicked)
    onButtonBaseDoubleClicked=(&self:WidgetTree.Button_Cancel.HandleOnButtonDoubleClicked)
    onButtonBaseHovered=(&self:WidgetTree.Button_Cancel.HandleOnButtonHovered)
    onButtonBaseUnhovered=(&self:WidgetTree.Button_Cancel.HandleOnButtonUnhovered)
    onSelectedChangedBase=(&self:WidgetTree.Button_Cancel.HandleOnSelectedChanged)
    triggeringInputAction {dataTable=&InputActionDataTable, rowName=Cancel}
   }
   slot {class=HorizontalBoxSlot}
  }
  Header {
   widgetClassPath=&Border
   parent=Overlay_2
   widget=Header
   properties {
    visibility=SelfHitTestInvisible
    background {
     bIsDynamicallyLoaded=false
     drawAs=Image
     imageType=NoImage
     mirroring=NoMirror
     resourceName=None
     resourceObject=/FortniteCoreUI/UI/Foundation/Materials/M_BoltGradientBorder_Creative_ServerSelect
     tiling=NoTile
     imageSize=<32 32>
     margin=<ltrb 0 0 0 0>
     outlineSettings {
      bUseBrushTransparency=false
      roundingType=HalfHeightRadius
      width=0
      color {colorUseRule=UseColor_Specified, specifiedColor=<rgba 0 0 0 0>}
      cornerRadii=<0 0 0 0>
     }
     tintColor {colorUseRule=UseColor_Specified, specifiedColor=<rgba 1 1 1 1>}
     uVRegion {bIsValid=false, max=<0 0>, min=<0 0>}
    }
    padding=<ltrb 0 0 0 0>
   }
   slot {class=OverlaySlot, properties {horizontalAlignment=HAlign_Fill}}
  }
  SafeZone_3 {
   widgetClassPath=&SafeZone
   bIsVariable=true
   parent=Header
   widget=SafeZone_3
   properties {padBottom=false}
   slot {class=BorderSlot, properties {padding=<ltrb 0 0 0 0>}}
  }
  HeaderArea {widgetClassPath=&HorizontalBox, parent=SafeZone_3, widget=HeaderArea, slot {class=SafeZoneSlot}}
  OverlayTitle {
   widgetClassPath=&Overlay
   bIsVariable=true
   parent=HeaderArea
   widget=OverlayTitle
   properties {renderTransformPivot=<0 0>}
   slot {class=HorizontalBoxSlot, properties {horizontalAlignment=HAlign_Left, verticalAlignment=VAlign_Center, padding=<ltrb 72 48 8 24>}}
  }
  SizeBox_3 {
   widgetClassPath=&SizeBox
   parent=OverlayTitle
   widget=SizeBox_3
   properties {
    bOverride_HeightOverride=true
    bOverride_WidthOverride=true
    heightOverride=100
    widthOverride=100
    renderTransform {angle=-15, scale=<1 1>, shear=<0 0>, translation=<-46 -34>}
   }
   slot {class=OverlaySlot}
  }
  Image_158 {
   widgetClassPath=&Image
   bIsVariable=true
   parent=SizeBox_3
   widget=Image_158
   properties {brush {
     bIsDynamicallyLoaded=false
     drawAs=Image
     imageType=NoImage
     mirroring=NoMirror
     resourceName=None
     resourceObject=/Game/Athena/HUD/Art/Materials/M_UI_CreativeLogo
     tiling=NoTile
     imageSize=<32 32>
     margin=<ltrb 0 0 0 0>
     outlineSettings {
      bUseBrushTransparency=false
      roundingType=HalfHeightRadius
      width=0
      color {colorUseRule=UseColor_Specified, specifiedColor=<rgba 0 0 0 0>}
      cornerRadii=<0 0 0 0>
     }
     tintColor {colorUseRule=UseColor_Specified, specifiedColor=<rgba 1 1 1 1>}
     uVRegion {bIsValid=false, max=<0 0>, min=<0 0>}
    }}
   slot {class=SizeBoxSlot}
  }
  VerticalBox_3 {widgetClassPath=&VerticalBox, parent=OverlayTitle, widget=VerticalBox_3, slot {class=OverlaySlot}}
  CommonTextBlock_101 {
   widgetClassPath=&CommonTextBlock
   parent=VerticalBox_3
   widget=CommonTextBlock_101
   properties {
    text=CREATIVE
    textTransformPolicy=ToUpper
    colorAndOpacity {colorUseRule=UseColor_Specified, specifiedColor=<rgba 0 1 1 1>}
    font {
     bForceMonospaced=false
     bMaterialIsStencil=false
     fontMaterial=null
     fontObject=&BurbankSmall
     letterSpacing=0
     monospacedWidth=1
     size=20
     skewAmount=0
     typefaceFontName=Black
     outlineSettings {
      bApplyOutlineToDropShadows=false
      bMiteredCorners=false
      bSeparateFillAlpha=false
      outlineMaterial=null
      outlineSize=0
      outlineColor=<rgba 0 0 0 1>
     }
    }
    renderTransform {angle=0, scale=<1 1>, shear=<-8 0>, translation=<0 0>}
    shadowOffset=<0 0>
   }
   slot {class=VerticalBoxSlot, properties {horizontalAlignment=HAlign_Left}}
  }
  SizeBox_0 {
   widgetClassPath=&SizeBox
   parent=VerticalBox_3
   widget=SizeBox_0
   properties {bOverride_WidthOverride=true, widthOverride=448}
   slot {class=VerticalBoxSlot}
  }
  ScaleBox_2 {
   widgetClassPath=&ScaleBox
   parent=SizeBox_0
   widget=ScaleBox_2
   properties {stretchDirection=DownOnly}
   slot {class=SizeBoxSlot, properties {horizontalAlignment=HAlign_Left}}
  }
  CommonTextBlock {
   widgetClassPath=&CommonTextBlock
   parent=ScaleBox_2
   widget=CommonTextBlock
   properties {
    text=Select A Server
    textTransformPolicy=ToUpper
    font {
     bForceMonospaced=false
     bMaterialIsStencil=false
     fontMaterial=null
     fontObject=&Burbank
     letterSpacing=0
     monospacedWidth=1
     size=48
     skewAmount=0
     typefaceFontName=Burbank-Black
     outlineSettings {
      bApplyOutlineToDropShadows=false
      bMiteredCorners=false
      bSeparateFillAlpha=false
      outlineMaterial=null
      outlineSize=0
      outlineColor=<rgba 0 0 0 1>
     }
    }
    renderTransform {angle=0, scale=<1 1>, shear=<-8 0>, translation=<0 -12>}
    renderTransformPivot=<1 0>
    shadowColorAndOpacity=<rgba 0 0 0 1>
    shadowOffset=<0 0>
   }
   slot {class=ScaleBoxSlot}
  }
  Spacer_253 {
   widgetClassPath=&Spacer
   parent=HeaderArea
   widget=Spacer_253
   slot {class=HorizontalBoxSlot, properties {size {sizeRule=Fill, value=1}}}
  }
  CreativeLobbyAdSpace {
   widgetClassPath=/FortniteCoreUI/UI/Creative/CreativeLobbyAdSpace
   bIsVariable=true
   parent=HeaderArea
   widget=CreativeLobbyAdSpace
   slot {class=HorizontalBoxSlot, properties {horizontalAlignment=HAlign_Center, verticalAlignment=VAlign_Center, padding=<ltrb 64 0 30 12>}}
  }
  VisibilityBorder_Touch {
   widgetClassPath=&CommonHardwareVisibilityBorder
   parent=HeaderArea
   widget=VisibilityBorder_Touch
   properties {visibilityQuery {
     autoDescription=" ANY( Input.Touch )"
     tokenStreamVersion=0
     userDescription=""
     queryTokenStream [0,1,1,1,0]
     tagDictionary [{tagName=Input.Touch}]
    }}
   slot {class=HorizontalBoxSlot, properties {horizontalAlignment=HAlign_Right, verticalAlignment=VAlign_Center, padding=<ltrb 0 0 16 16>}}
  }
  Button_CloseTouch {
   widgetClassPath=/FortUILibrary/Foundation/Buttons/CloseButton
   bInherited=true
   bIsVariable=true
   parent=VisibilityBorder_Touch
   widget=Button_CloseTouch
   properties {
    onButtonBaseClicked=(&self:WidgetTree.Button_CloseTouch.HandleOnButtonClicked)
    onButtonBaseDoubleClicked=(&self:WidgetTree.Button_CloseTouch.HandleOnButtonDoubleClicked)
    onButtonBaseHovered=(&self:WidgetTree.Button_CloseTouch.HandleOnButtonHovered)
    onButtonBaseUnhovered=(&self:WidgetTree.Button_CloseTouch.HandleOnButtonUnhovered)
    onSelectedChangedBase=(&self:WidgetTree.Button_CloseTouch.HandleOnSelectedChanged)
   }
   slot {class=BorderSlot, properties {padding=<ltrb 0 0 0 0>}}
  }
 ]
}
